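; PCG subgraph that cuts a spline into separate linear splines by dropping segments over DistanceThreshold, or under it with InvertFilter; Passthrough returns the input untouched.
; It samples points whose BoundsMax.x encodes half the segment length, then flags those to keep.
; Original index minus filtered index identifies contiguous kept runs.
; The first rejected point after each run rejoins it as endpoint, then points are sorted, partitioned by Delta and rebuilt.
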(graph CutSplines
  (params (DistanceThreshold :type Double :default 3e+03) (InvertFilter :type Bool) (Passthrough :type Bool))
  (inputs (SplineIn :type Spline))
  (outputs (SplinesOut :type Spline))
  (bind Branch_29 (Branch :bOutputToB-param Passthrough))
  (bind SplineSampler_14 (Spline%20Sampler :SamplerParams (struct :SubdivisionsPerSegment 0 :DistanceIncrement 2e+02 :bUnbounded true :PointSteepness 1.0)))
  (bind GetPointsCount_2 (Get%20Element%20Count))
  (bind Branch_3 (Branch))
  (bind Branch_40 (Branch :bOutputToB-param InvertFilter))
  (bind AttributeCompareOp_5 (Attribute%20Compare%20Op :InputSource1 $BoundsMax.x :InputSource2 @Last :Operation Less :OutputTarget Keep))
  (bind AttributeCompareOp_36 (Attribute%20Compare%20Op :InputSource1 $BoundsMax.x :InputSource2 @Last :Operation Greater :OutputTarget Keep))
  (bind CopyAttribute_9 (Copy%20Attributes :InputSource $Index :OutputTarget indexbck))
  (bind PointFilter_8 (Filter%20Attribute%20Elements :AttributeTypes (struct :Type Boolean :BoolValue true) :bUseConstantThreshold true :bUseSpatialQuery true :Operator Equal :TargetAttribute Keep :ThresholdAttribute ThresholdFailed))
  (bind AttributeMathsOp_10 (Attribute%20Maths%20Op :InputSource1 indexbck :InputSource2 $Index :Operation Subtract :OutputTarget Delta))
  (bind AttributeMathsOp_12 (Attribute%20Maths%20Op :InputSource1 indexbck :InputSource2 $Index :Operation Subtract :OutputTarget Delta))
  (bind FilterDataByAttribute_0 (Filter%20Data%20By%20Attribute :Attribute Delta))
  (bind AttributePartition_13 (Attribute%20Partition :bTokenizeOnWhiteSpace true :PartitionAttributeSelectors [Delta]))
  (bind PointFilter_14 (Filter%20Attribute%20Elements :AttributeTypes (struct :Type Integer32) :bUseConstantThreshold true :bUseSpatialQuery true :Operator Equal :TargetAttribute $Index))
  (bind Merge_22 (Merge%20Points))
  (bind AttributeMathsOp_21 (Attribute%20Maths%20Op :InputSource1 indexbck :InputSource2 delta :Operation Subtract :OutputTarget Delta))
  (bind Merge_15 (Merge%20Points))
  (bind SortPoints_20 (Sort%20Attributes :bUseStableSort false :SortAttributes [(struct :InputSource indexbck)]))
  (bind AttributePartition_16 (Attribute%20Partition :bTokenizeOnWhiteSpace true :PartitionAttributeSelectors [Delta]))
  (bind CreateSpline_48 (Create%20Spline :ArriveTangentAttribute None :bLinear true :InterpTypeAttribute None :LeaveTangentAttribute None))
  (bind Gather_32 (Gather))
  ;[0 0 0] BoundsMax.x = half segment
  (bind CreateAttribute_1 (Create%20Constant :AttributeTypes (struct :DoubleValue 0.5)))
  (bind AttributeMathsOp_0 (Attribute%20Maths%20Op :Operation Multiply :InA DistanceThreshold))
  (connect Branch_40 Output%20A AttributeCompareOp_5 InA)
  (connect AttributeMathsOp_0 Out AttributeCompareOp_5 InB)
  (connect Branch_40 Output%20B AttributeCompareOp_36 InA)
  (connect AttributeMathsOp_0 Out AttributeCompareOp_36 InB)
  (connect CreateAttribute_1 Out AttributeMathsOp_0 InB)
  (connect PointFilter_8 InsideFilter AttributeMathsOp_10 InA)
  (connect PointFilter_8 InsideFilter AttributeMathsOp_10 InB)
  (connect PointFilter_8 OutsideFilter AttributeMathsOp_12 InA)
  (connect PointFilter_8 OutsideFilter AttributeMathsOp_12 InB)
  (connect Merge_22 Out AttributeMathsOp_21 InA)
  (connect Merge_22 Out AttributeMathsOp_21 InB)
  (connect FilterDataByAttribute_0 InsideFilter AttributePartition_13 In)
  (connect SortPoints_20 Out AttributePartition_16 In)
  (connect GetPointsCount_2 Out Branch_3 bOutputToB)
  (connect SplineSampler_14 Out Branch_3 In)
  (connect DefaultInputNode SplineIn Branch_29 In)
  (connect Branch_3 Output%20B Branch_40 In)
  (connect AttributeCompareOp_5 Out CopyAttribute_9 Source)
  (connect AttributeCompareOp_36 Out CopyAttribute_9 Source)
  (connect AttributeCompareOp_5 Out CopyAttribute_9 Target)
  (connect AttributeCompareOp_36 Out CopyAttribute_9 Target)
  (connect AttributePartition_16 Out CreateSpline_48 In)
  (connect AttributeMathsOp_12 Out FilterDataByAttribute_0 In)
  (connect Branch_29 Output%20B Gather_32 In)
  (connect CreateSpline_48 Out Gather_32 In)
  (connect SplineSampler_14 Out GetPointsCount_2 In)
  (connect AttributeMathsOp_10 Out Merge_15 In)
  (connect AttributeMathsOp_21 Out Merge_15 In)
  (connect PointFilter_14 InsideFilter Merge_22 In)
  (connect CopyAttribute_9 Out PointFilter_8 In)
  (connect AttributePartition_13 Out PointFilter_14 In)
  (connect Merge_15 Out SortPoints_20 In)
  (connect Branch_29 Output%20A SplineSampler_14 Spline)
  (output SplinesOut Gather_32))
