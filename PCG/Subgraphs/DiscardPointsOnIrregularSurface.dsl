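; PCG subgraph that removes input points sitting on uneven parts of a surface.
; Four corner samples per point, sized by HeightThreshold and spaced using the first point's footprint and ExtentsMultiplier, are projected onto the surface.
; Samples missing a thin plane at the original point are flagged, alongside a normal-variance test intersecting projected slabs with an offset plane.
; Flagged results are expanded and subtracted from the input via Difference.
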(graph DiscardPointsOnIrregularSurface
  (params (HeightThreshold :type Double :default 15.0) (ExtentsMultiplier :type Double :default 1.0))
  (inputs (Points%20In :type Point) (Surface :type Surface))
  (outputs (Points%20Out :type Point))
  (bind ApplyScaleToBounds_0 (Apply%20Scale%20To%20Bounds))
  (bind GetAttributeFromPointIndex_10 (Get%20Attribute%20From%20Point%20Index :bForceOutputAttributeToBeInElementsDomain true :InputSource $BoundsMax :OutputAttributeName BMax))
  (bind GetAttributeFromPointIndex_12 (Get%20Attribute%20From%20Point%20Index :bForceOutputAttributeToBeInElementsDomain true :InputSource $BoundsMin :OutputAttributeName BMin))
  (bind AttributeMathsOp_11 (Attribute%20Maths%20Op :InputSource1 @LastCreated :InputSource2 @LastCreated :InputSource3 @LastCreated :Operation Subtract :OutputTarget @SourceName))
  (bind CreateAttribute_21 (Create%20Constant :AttributeTypes (struct :Type Vector :VectorValue (1.0 1.0 0.0))))
  (bind AttributeMathsOp_20 (Attribute%20Maths%20Op :InputSource1 @LastCreated :InputSource2 @LastCreated :InputSource3 @LastCreated :Operation Multiply :OutputTarget @SourceName))
  (bind AttributeVectorOp_13 (Attribute%20Vector%20Op :InputSource1 @LastCreated :InputSource2 @LastCreated :InputSource3 @LastCreated :Operation Length :OutputTarget @SourceName))
  (bind AttributeMathsOp_27 (Attribute%20Maths%20Op :InputSource1 @LastCreated :InputSource2 @LastCreated :InputSource3 @LastCreated :Operation Multiply :OutputTarget @SourceName :InA ExtentsMultiplier))
  (bind CreateAttribute_19 (Create%20Constant :AttributeTypes (struct :DoubleValue 0.5)))
  (bind AttributeMathsOp_17 (Attribute%20Maths%20Op :InputSource1 @LastCreated :InputSource2 @LastCreated :InputSource3 @LastCreated :Operation Multiply :OutputTarget @SourceName))
  (bind DensityRemap_4 (Density%20Remap :OutRangeMin 1.0))
  (bind CreateAttribute_27 (Create%20Constant :AttributeTypes (struct :DoubleValue -1.0)))
  (bind Make%20Vector%20Attribute_25 (Make%20Vector%20Attribute :InputSource1 @LastCreated :InputSource2 @LastCreated :InputSource3 @LastCreated :InputSource4 @LastCreated :OutputTarget @SourceName :OutputType Vector :X HeightThreshold :Y HeightThreshold :Z HeightThreshold))
  (bind Attribute%20Maths%20Op_28 (Attribute%20Maths%20Op :InputSource1 @LastCreated :InputSource2 @LastCreated :InputSource3 @LastCreated :Operation Multiply :OutputTarget @SourceName))
  (bind Union_8 (Union :DensityFunction Binary :Type KeepAll))
  (bind CreateAttribute_25 (Create%20Constant :AttributeTypes (struct :DoubleValue -1.0)))
  (bind AttributeMathsOp_22 (Attribute%20Maths%20Op :InputSource1 @LastCreated :InputSource2 @LastCreated :InputSource3 @LastCreated :Operation Multiply :OutputTarget @SourceName))
  (bind Union_2 (Union :DensityFunction Binary :Type KeepAll))
  (bind Union_0 (Union)) ; Node added to replicate deprecated behavior of the Difference node.
  (bind ToPoint_0 (To%20Point))
  ;[0 0 0] Expand out-of-plane points + points with high normal variance before diff
  (bind CreateAttribute_35 (Create%20Constant :AttributeTypes (struct :DoubleValue -1.0)))
  (bind Make%20Vector%20Attribute_22 (Make%20Vector%20Attribute :InputSource1 @LastCreated :InputSource2 @LastCreated :InputSource3 @LastCreated :InputSource4 @LastCreated :OutputTarget @SourceName :OutputType Vector))
  (bind Attribute%20Maths%20Op_24 (Attribute%20Maths%20Op :InputSource1 @LastCreated :InputSource2 @LastCreated :InputSource3 @LastCreated :Operation Multiply :OutputTarget @SourceName))
  (bind BoundsModifier_28 (Bounds%20Modifier :bAffectSteepness true :BoundsMax (5e+01 5e+01 5e+01) :BoundsMin (-5e+01 -5e+01 -5e+01) :Mode Set))
  ;[0 0 0] Build bounds for plane (thin box) centered at projected point location
  (bind CreateAttribute_29 (Create%20Constant :AttributeTypes (struct :DoubleValue -1.0)))
  (bind CreateAttribute_33 (Create%20Constant :AttributeTypes (struct :DoubleValue 1.5)))
  (bind Attribute%20Maths%20Op_12 (Attribute%20Maths%20Op :InputSource1 @LastCreated :InputSource2 @LastCreated :InputSource3 @LastCreated :Operation Multiply :OutputTarget @SourceName))
  (bind CreateAttribute_31 (Create%20Constant :AttributeTypes (struct :DoubleValue 1.0)))
  (bind Make%20Vector%20Attribute_2 (Make%20Vector%20Attribute :InputSource1 @LastCreated :InputSource2 @LastCreated :InputSource3 @LastCreated :InputSource4 @LastCreated :OutputTarget @SourceName :OutputType Vector))
  (bind Attribute%20Maths%20Op_8 (Attribute%20Maths%20Op :InputSource1 @LastCreated :InputSource2 @LastCreated :InputSource3 @LastCreated :Operation Multiply :OutputTarget @SourceName))
  ;[0 0 0] Project the 4 sample points on the surface (landscape), diff them against a 2d plane at Input point location to keep points that aren't intersecting the plane
  (bind BoundsModifier_7 (Bounds%20Modifier :bAffectSteepness true :BoundsMax (2e+02 2e+02 1.0) :BoundsMin (-2e+02 -2e+02 -1.0) :Mode Set))
  (bind Projection_5 (Projection))
  (bind Union_1 (Union)) ; Node added to replicate deprecated behavior of the Difference node.
  (bind Difference_3 (Difference :bDiffMetadata false :DensityFunction Binary))
  ;[0 0 0] Generate 4 points in a 2D square around each input Point
  (bind BoundsModifier_6 (Bounds%20Modifier :bAffectSteepness true :BoundsMax (1e+01 1e+01 1e+01) :BoundsMin (-1e+01 -1e+01 -1e+01) :Mode Set))
  (bind CreateAttribute_18 (Create%20Constant))
  (bind Make%20Vector%20Attribute_1 (Make%20Vector%20Attribute :InputSource1 @LastCreated :InputSource2 @LastCreated :InputSource3 @LastCreated :InputSource4 @LastCreated :OutputTarget @SourceName :OutputType Vector))
  (bind Transform%20Points_7 (Transform%20Points :OffsetMax (2e+02 2e+02 0.0) :OffsetMin (2e+02 2e+02 0.0)))
  (bind CreateAttribute_20 (Create%20Constant :AttributeTypes (struct :Type Vector :VectorValue (-1.0 1.0 1.0))))
  (bind Attribute%20Maths%20Op_6 (Attribute%20Maths%20Op :InputSource1 @LastCreated :InputSource2 @LastCreated :InputSource3 @LastCreated :Operation Multiply :OutputTarget @SourceName))
  (bind Transform%20Points_8 (Transform%20Points :OffsetMax (-2e+02 2e+02 0.0) :OffsetMin (-2e+02 2e+02 0.0)))
  (bind CreateAttribute_22 (Create%20Constant :AttributeTypes (struct :Type Vector :VectorValue (1.0 -1.0 1.0))))
  (bind Attribute%20Maths%20Op_9 (Attribute%20Maths%20Op :InputSource1 @LastCreated :InputSource2 @LastCreated :InputSource3 @LastCreated :Operation Multiply :OutputTarget @SourceName))
  (bind Transform%20Points_9 (Transform%20Points :OffsetMax (2e+02 -2e+02 0.0) :OffsetMin (2e+02 -2e+02 0.0)))
  (bind CreateAttribute_24 (Create%20Constant :AttributeTypes (struct :Type Vector :VectorValue (-1.0 -1.0 1.0))))
  (bind Attribute%20Maths%20Op_13 (Attribute%20Maths%20Op :InputSource1 @LastCreated :InputSource2 @LastCreated :InputSource3 @LastCreated :Operation Multiply :OutputTarget @SourceName))
  (bind Transform%20Points_10 (Transform%20Points :OffsetMax (-2e+02 -2e+02 0.0) :OffsetMin (-2e+02 -2e+02 0.0)))
  ;[0 0 0] Normal Variance Test
  (bind CreateAttribute_16 (Create%20Constant :AttributeTypes (struct :DoubleValue -1.0)))
  (bind CreateAttribute_12 (Create%20Constant))
  (bind Make%20Vector%20Attribute_7 (Make%20Vector%20Attribute :InputSource1 @LastCreated :InputSource2 @LastCreated :InputSource3 @LastCreated :InputSource4 @LastCreated :OutputTarget @SourceName :OutputType Vector))
  (bind Attribute%20Maths%20Op_3 (Attribute%20Maths%20Op :InputSource1 @LastCreated :InputSource2 @LastCreated :InputSource3 @LastCreated :Operation Multiply :OutputTarget @SourceName))
  (bind BoundsModifier_2 (Bounds%20Modifier :BoundsMax (2e+02 2e+02 1.0) :BoundsMin (-2e+02 -2e+02 -1.0) :Mode Set))
  (bind Transform%20Points_0 (Transform%20Points :OffsetMax (0.0 0.0 1.0) :OffsetMin (0.0 0.0 1.0)))
  (bind Intersection_1 (Inner%20Intersection))
  ;;[0 0 0] Debug Normal Threshold
  (bind Transform%20Points_1 (Transform%20Points))
  ;[0 0 0] Diff the input against the out-of-plane  points
  (bind Difference_0 (Difference :bDiffMetadata false :DensityFunction Binary))
  (connect DefaultInputNode Points%20In ApplyScaleToBounds_0 In)
  (connect Make%20Vector%20Attribute_7 Out Attribute%20Maths%20Op_3 InA)
  (connect CreateAttribute_16 Out Attribute%20Maths%20Op_3 InB)
  (connect Make%20Vector%20Attribute_1 Out Attribute%20Maths%20Op_6 InA)
  (connect CreateAttribute_20 Out Attribute%20Maths%20Op_6 InB)
  (connect Make%20Vector%20Attribute_2 Out Attribute%20Maths%20Op_8 InA)
  (connect CreateAttribute_29 Out Attribute%20Maths%20Op_8 InB)
  (connect Make%20Vector%20Attribute_1 Out Attribute%20Maths%20Op_9 InA)
  (connect CreateAttribute_22 Out Attribute%20Maths%20Op_9 InB)
  (connect AttributeMathsOp_17 Out Attribute%20Maths%20Op_12 InA)
  (connect CreateAttribute_33 Out Attribute%20Maths%20Op_12 InB)
  (connect Make%20Vector%20Attribute_1 Out Attribute%20Maths%20Op_13 InA)
  (connect CreateAttribute_24 Out Attribute%20Maths%20Op_13 InB)
  (connect Make%20Vector%20Attribute_22 Out Attribute%20Maths%20Op_24 InA)
  (connect CreateAttribute_35 Out Attribute%20Maths%20Op_24 InB)
  (connect Make%20Vector%20Attribute_25 Out Attribute%20Maths%20Op_28 InA)
  (connect CreateAttribute_27 Out Attribute%20Maths%20Op_28 InB)
  (connect GetAttributeFromPointIndex_10 Attribute AttributeMathsOp_11 InA)
  (connect GetAttributeFromPointIndex_12 Attribute AttributeMathsOp_11 InB)
  (connect AttributeMathsOp_27 Out AttributeMathsOp_17 InA)
  (connect CreateAttribute_19 Out AttributeMathsOp_17 InB)
  (connect AttributeMathsOp_11 Out AttributeMathsOp_20 InA)
  (connect CreateAttribute_21 Out AttributeMathsOp_20 InB)
  (connect AttributeMathsOp_27 Out AttributeMathsOp_22 InA)
  (connect CreateAttribute_25 Out AttributeMathsOp_22 InB)
  (connect AttributeVectorOp_13 Out AttributeMathsOp_27 InB)
  (connect AttributeMathsOp_20 Out AttributeVectorOp_13 In)
  (connect Make%20Vector%20Attribute_2 Out BoundsModifier_2 BoundsMax)
  (connect Attribute%20Maths%20Op_8 Out BoundsModifier_2 BoundsMin)
  (connect Projection_5 Out BoundsModifier_2 In)
  (connect Make%20Vector%20Attribute_25 Out BoundsModifier_6 BoundsMax)
  (connect Attribute%20Maths%20Op_28 Out BoundsModifier_6 BoundsMin)
  (connect DensityRemap_4 Out BoundsModifier_6 In)
  (connect Make%20Vector%20Attribute_2 Out BoundsModifier_7 BoundsMax)
  (connect Attribute%20Maths%20Op_8 Out BoundsModifier_7 BoundsMin)
  (connect DensityRemap_4 Out BoundsModifier_7 In)
  (connect Make%20Vector%20Attribute_22 Out BoundsModifier_28 BoundsMax)
  (connect Attribute%20Maths%20Op_24 Out BoundsModifier_28 BoundsMin)
  (connect Union_2 Out BoundsModifier_28 In)
  (connect DefaultInputNode Points%20In DensityRemap_4 In)
  (connect BoundsModifier_28 Out Difference_0 Differences)
  (connect Union_0 Out Difference_0 Source)
  (connect BoundsModifier_7 Out Difference_3 Differences)
  (connect Union_1 Out Difference_3 Source)
  (connect ApplyScaleToBounds_0 Out GetAttributeFromPointIndex_10 In)
  (connect ApplyScaleToBounds_0 Out GetAttributeFromPointIndex_12 In)
  (connect BoundsModifier_2 Out Intersection_1 In)
  (connect Transform%20Points_0 Out Intersection_1 In)
  (connect AttributeMathsOp_17 Out Make%20Vector%20Attribute_1 X)
  (connect AttributeMathsOp_17 Out Make%20Vector%20Attribute_1 Y)
  (connect CreateAttribute_18 Out Make%20Vector%20Attribute_1 Z)
  (connect Attribute%20Maths%20Op_12 Out Make%20Vector%20Attribute_2 X)
  (connect Attribute%20Maths%20Op_12 Out Make%20Vector%20Attribute_2 Y)
  (connect CreateAttribute_31 Out Make%20Vector%20Attribute_2 Z)
  (connect CreateAttribute_12 Out Make%20Vector%20Attribute_7 X)
  (connect CreateAttribute_12 Out Make%20Vector%20Attribute_7 Y)
  (connect AttributeMathsOp_22 Out Make%20Vector%20Attribute_7 Z)
  (connect AttributeMathsOp_17 Out Make%20Vector%20Attribute_22 X)
  (connect AttributeMathsOp_17 Out Make%20Vector%20Attribute_22 Y)
  (connect AttributeMathsOp_17 Out Make%20Vector%20Attribute_22 Z)
  (connect Union_8 Out Projection_5 In)
  (connect DefaultInputNode Surface Projection_5 Projection%20Target)
  (connect Difference_0 Out ToPoint_0 In)
  (connect BoundsModifier_7 Out Transform%20Points_0 In)
  (connect Make%20Vector%20Attribute_7 Out Transform%20Points_0 OffsetMax)
  (connect Make%20Vector%20Attribute_7 Out Transform%20Points_0 OffsetMin)
  (connect Intersection_1 Out Transform%20Points_1 In)
  (connect Attribute%20Maths%20Op_3 Out Transform%20Points_1 OffsetMax)
  (connect Attribute%20Maths%20Op_3 Out Transform%20Points_1 OffsetMin)
  (connect BoundsModifier_6 Out Transform%20Points_7 In)
  (connect Make%20Vector%20Attribute_1 Out Transform%20Points_7 OffsetMax)
  (connect Make%20Vector%20Attribute_1 Out Transform%20Points_7 OffsetMin)
  (connect BoundsModifier_6 Out Transform%20Points_8 In)
  (connect Attribute%20Maths%20Op_6 Out Transform%20Points_8 OffsetMax)
  (connect Attribute%20Maths%20Op_6 Out Transform%20Points_8 OffsetMin)
  (connect BoundsModifier_6 Out Transform%20Points_9 In)
  (connect Attribute%20Maths%20Op_9 Out Transform%20Points_9 OffsetMax)
  (connect Attribute%20Maths%20Op_9 Out Transform%20Points_9 OffsetMin)
  (connect BoundsModifier_6 Out Transform%20Points_10 In)
  (connect Attribute%20Maths%20Op_13 Out Transform%20Points_10 OffsetMax)
  (connect Attribute%20Maths%20Op_13 Out Transform%20Points_10 OffsetMin)
  (connect DefaultInputNode Points%20In Union_0 In)
  (connect Projection_5 Out Union_1 In)
  (connect Difference_3 Out Union_2 In)
  (connect Transform%20Points_1 Out Union_2 In)
  (connect Transform%20Points_7 Out Union_8 In)
  (connect Transform%20Points_8 Out Union_8 In)
  (connect Transform%20Points_9 Out Union_8 In)
  (connect Transform%20Points_10 Out Union_8 In)
  (output Points%20Out ToPoint_0))
