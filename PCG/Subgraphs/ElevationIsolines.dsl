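; PCG subgraph that generates points along elevation isolines of a surface, at a set vertical spacing between minimum and maximum heights.
; Bounding-shape or owning-actor bounds become a height-adjusted volume, optionally diced into grid cells and intersected with surface and shape.
; ElevationIsolinesInnerLoop traces lines per cell from line count and spline sampling settings.
; Results merge, optionally self-prune per Z level, clip to the original volume when sliced, and receive uniform extents and density.
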(graph ElevationIsolines
  (params (SplineSamplingDistance :type Double :default 2.5e+02) (VerticalSpacing :type Int :default 500) (HeightMin :type Double) (HeightMax :type Double :default 5e+03) (SliceInputPoint :type Bool :default true) (SliceSize :type Double :default 5e+03) (DisplayInputBounds :type Bool) (ZUp :type Bool :default true) (IgnorePCGHits :type Bool :default true) (CollisionChannel :type (Enum /Script/Engine.ECollisionChannel)) (ColorSplines :type Bool) (ExperimentalResampling :type Bool) (SelfPruning :type Bool :default true) (DisplayPoints :type Bool))
  (inputs (Surface :type Surface) (BoundingShape :type Spatial))
  (outputs (Points :type Point))
  ;[0 0 0] Slice and Dice Input Point if required
  (bind Branch_28 (Branch :bOutputToB-param SliceInputPoint))
  (bind GetAttributeFromPointIndex_0 (Get%20Attribute%20From%20Point%20Index :bForceOutputAttributeToBeInElementsDomain true :InputSource $Extents :OutputAttributeName None))
  (bind BreakVectorAttribute_2 (Break%20Vector%20Attribute))
  (bind CreateAttribute_3 (Create%20Constant :AttributeTypes (struct :DoubleValue 2.0)))
  (bind AttributeMathsOp_2 (Attribute%20Maths%20Op :Operation Multiply))
  (bind MakeVectorAttribute_3 (Make%20Vector%20Attribute :OutputType Vector :X SliceSize :Y SliceSize))
  (bind AttributeMathsOp_1 (Attribute%20Maths%20Op :Operation Divide))
  (bind AttributeMathsOp_3 (Attribute%20Maths%20Op :InputSource1 @Last.xy :Operation Ceil))
  (bind AttributeMathsOp_10 (Attribute%20Maths%20Op :Operation Multiply))
  (bind CreatePointsGrid_1 (Create%20Points%20Grid :PointSteepness 0.5))
  (bind AttributeTransformOp_0 (Attribute%20Transform%20Op :InputSource1 $Transform :InputSource2 $Transform))
  (bind AttributePartition_10 (Attribute%20Partition :bTokenizeOnWhiteSpace true :PartitionAttributeSelectors [$Seed]))
  (bind GetAttributeFromPointIndex_20 (Get%20Attribute%20From%20Point%20Index :bForceOutputAttributeToBeInElementsDomain true :InputSource $Position :OutputAttributeName None))
  (bind Branch_26 (Branch :bOutputToB-param DisplayInputBounds))
  (bind Debug_25 (Debug))
  (bind CreateAttribute_2 (Create%20Constant :AttributeTypes (struct :DoubleValue 1.0)))
  (bind CopyAttribute_0 (Copy%20Attributes :InputSource @Last :OutputTarget $Steepness))
  (bind Gather_35 (Gather))
  (bind Intersection_2 (Intersection :bIgnorePinsWithNoInput true))
  (bind ToPoint_4 (To%20Point))
  ;[0 0 0] Generate Isolines
  (bind CreateAttribute_1 (Create%20Constant :AttributeTypes (struct :DoubleValue 1e+01)))
  (bind AttributeMathsOp_0 (Attribute%20Maths%20Op :Operation Max :InA SplineSamplingDistance))
  (bind CreateAttribute_8 (Create%20Constant :AttributeTypes (struct :DoubleValue 0.5)))
  (bind AttributeMathsOp_7 (Attribute%20Maths%20Op :Operation Multiply))
  (bind Loop_12 (Loop :bTokenizeOnWhiteSpace true :Graph /PCG/Subgraphs/ElevationIsolinesInnerLoop.ElevationIsolinesInnerLoop :CollisionChannel CollisionChannel :ExperimentalResampling ExperimentalResampling :IgnorePCGHits IgnorePCGHits :SetSplineColors ColorSplines :ZUp ZUp))
  ;[0 0 0] Prepare Output
  (bind Merge_16 (Merge%20Points))
  (bind AttributePartition_0 (Attribute%20Partition :bTokenizeOnWhiteSpace true :PartitionAttributeSelectors [$Position.z]))
  (bind Branch_17 (Branch :bOutputToB-param SelfPruning))
  (bind Branch_2 (Branch :bOutputToB-param ExperimentalResampling))
  (bind ExtentsModifier_0 (Extents%20Modifier :Extents (0.5 1.0 1.0) :Mode Multiply))
  (bind SelfPruning_17 (Self%20Pruning :Parameters (struct :PruningType AllEqual :ComparisonSource $Extents :bRandomizedPruning false)))
  (bind Branch_15 (Branch :bOutputToB-param SliceInputPoint))
  (bind CreateAttribute_28 (Create%20Constant :AttributeTypes (struct :DoubleValue 0.5)))
  (bind AttributeMathsOp_26 (Attribute%20Maths%20Op :Operation Multiply :InA SplineSamplingDistance))
  (bind CreateAttribute_24 (Create%20Constant :AttributeTypes (struct :DoubleValue 1e+01)))
  (bind MakeVectorAttribute_22 (Make%20Vector%20Attribute :OutputType Vector))
  (bind Intersection_17 (Intersection :DensityFunction Minimum))
  (bind ToPoint_19 (To%20Point))
  (bind ExtentsModifier_18 (Extents%20Modifier :Extents (1.0 25.0 25.0)))
  (bind CreateAttribute_47 (Create%20Constant :AttributeTypes (struct :DoubleValue 1.0)))
  (bind AttributeMathsOp_46 (Attribute%20Maths%20Op :InputSource1 $Density :Operation Set))
  (bind Branch_44 (Branch :bOutputToB-param DisplayPoints))
  (bind ExtentsModifier_45 (Extents%20Modifier :Extents (-1.0 -1.0 -1.0) :Mode Add))
  (bind Debug_42 (Debug))
  ;[0 0 0] Prepare Input Point From Actor or BoundingShape
  (bind CreateAttribute_7 (Create%20Constant))
  (bind DataNum_2 (Data%20Count))
  (bind GetActorData_11 (Get%20Actor%20Data :Mode GetSinglePoint))
  (bind Branch_9 (Branch))
  (bind ApplyScaleToBounds_14 (Apply%20Scale%20To%20Bounds))
  (bind Branch_3 (Branch))
  (bind Union_0 (Union))
  (bind Subgraph_1 (Subgraph :Graph /PCG/Subgraphs/SpatialDataBoundsToPoint.SpatialDataBoundsToPoint))
  (bind Gather_12 (Gather))
  (bind AttributeMathsOp_6 (Attribute%20Maths%20Op :InputSource1 $BoundsMin.z :Operation Set))
  (bind AttributeMathsOp_4 (Attribute%20Maths%20Op :InputSource1 $Position.z :Operation Set :InB HeightMin))
  (bind AttributeMathsOp_9 (Attribute%20Maths%20Op :Operation Subtract :InA HeightMax :InB HeightMin))
  (bind AttributeMathsOp_5 (Attribute%20Maths%20Op :InputSource1 $BoundsMax.z :Operation Set))
  (bind AttributeMathsOp_8 (Attribute%20Maths%20Op :Operation Divide :InB VerticalSpacing))
  (bind ResetPointCenter_10 (Reset%20Point%20Center))
  (reroute ElevationIsolinesPoints Loop_12.OutPoints)
  (reroute FinalizedPoint ToPoint_4)
  (reroute InputPoint ResetPointCenter_10)
  (reroute InputSurface Surface)
  (reroute LineCount AttributeMathsOp_8)
  (reroute Reroute_1 BoundingShape)
  (connect Branch_9 Output%20A ApplyScaleToBounds_14 In)
  (connect CreateAttribute_1 Out AttributeMathsOp_0 InB)
  (connect GetAttributeFromPointIndex_0 Attribute AttributeMathsOp_1 InA)
  (connect MakeVectorAttribute_3 Out AttributeMathsOp_1 InB)
  (connect BreakVectorAttribute_2 Z AttributeMathsOp_2 InA)
  (connect CreateAttribute_3 Out AttributeMathsOp_2 InB)
  (connect AttributeMathsOp_1 Out AttributeMathsOp_3 In)
  (connect AttributeMathsOp_6 Out AttributeMathsOp_4 InA)
  (connect AttributeMathsOp_4 Out AttributeMathsOp_5 InA)
  (connect AttributeMathsOp_9 Out AttributeMathsOp_5 InB)
  (connect Gather_12 Out AttributeMathsOp_6 InA)
  (connect CreateAttribute_7 Out AttributeMathsOp_6 InB)
  (connect AttributeMathsOp_0 Out AttributeMathsOp_7 InA)
  (connect CreateAttribute_8 Out AttributeMathsOp_7 InB)
  (connect AttributeMathsOp_9 Out AttributeMathsOp_8 InA)
  (connect AttributeMathsOp_3 Out AttributeMathsOp_10 InA)
  (connect MakeVectorAttribute_3 Out AttributeMathsOp_10 InB)
  (connect CreateAttribute_28 Out AttributeMathsOp_26 InB)
  (connect ExtentsModifier_18 Out AttributeMathsOp_46 InA)
  (connect CreateAttribute_47 Out AttributeMathsOp_46 InB)
  (connect Merge_16 Out AttributePartition_0 In)
  (connect AttributeTransformOp_0 Out AttributePartition_10 In)
  (connect CreatePointsGrid_1 Out AttributeTransformOp_0 InA)
  (connect GetAttributeFromPointIndex_0 Point AttributeTransformOp_0 InB)
  (connect Branch_17 Output%20B Branch_2 In)
  (connect DataNum_2 Out Branch_3 bOutputToB)
  (connect DefaultInputNode BoundingShape Branch_3 In)
  (connect DataNum_2 Out Branch_9 bOutputToB)
  (connect GetActorData_11 Out Branch_9 In)
  (connect Branch_17 Output%20A Branch_15 In)
  (connect SelfPruning_17 Out Branch_15 In)
  (connect AttributePartition_0 Out Branch_17 In)
  (connect FinalizedPoint Out Branch_26 In)
  (connect InputPoint Out Branch_28 In)
  (connect AttributeMathsOp_46 Out Branch_44 In)
  (connect GetAttributeFromPointIndex_0 Attribute BreakVectorAttribute_2 In)
  (connect CreateAttribute_2 Out CopyAttribute_0 Source)
  (connect AttributePartition_10 Out CopyAttribute_0 Target)
  (connect MakeVectorAttribute_3 Out CreatePointsGrid_1 CellSize)
  (connect AttributeMathsOp_10 Out CreatePointsGrid_1 GridExtents)
  (connect DefaultInputNode BoundingShape DataNum_2 In)
  (connect Branch_26 Output%20B Debug_25 In)
  (connect ExtentsModifier_45 Out Debug_42 In)
  (connect Branch_2 Output%20B ExtentsModifier_0 In)
  (connect MakeVectorAttribute_22 Out ExtentsModifier_18 Extents)
  (connect Branch_15 Output%20A ExtentsModifier_18 In)
  (connect ToPoint_19 Out ExtentsModifier_18 In)
  (connect Branch_44 Output%20B ExtentsModifier_45 In)
  (connect ApplyScaleToBounds_14 Out Gather_12 In)
  (connect Subgraph_1 Point Gather_12 In)
  (connect CopyAttribute_0 Out Gather_35 In)
  (connect GetAttributeFromPointIndex_20 Point Gather_35 In)
  (connect Branch_28 Output%20B GetAttributeFromPointIndex_0 In)
  (connect Branch_28 Output%20A GetAttributeFromPointIndex_20 In)
  (connect Gather_35 Out Intersection_2 Primary%20Source)
  (connect Reroute_1 Out Intersection_2 Source1)
  (connect InputSurface Out Intersection_2 Source2)
  (connect Branch_15 Output%20B Intersection_17 Primary%20Source)
  (connect InputPoint Out Intersection_17 Source1)
  (connect AttributeMathsOp_0 Out Loop_12 ExperimentalResamplingCutSplinesDistThreshold)
  (connect AttributeMathsOp_7 Out Loop_12 ExperimentalResamplingInternalDistance)
  (connect FinalizedPoint Out Loop_12 InPoint)
  (connect LineCount Out Loop_12 lineCount)
  (connect AttributeMathsOp_0 Out Loop_12 SplineSamplingDistance)
  (connect InputSurface Out Loop_12 Surface)
  (connect AttributeMathsOp_2 Out MakeVectorAttribute_3 Z)
  (connect AttributeMathsOp_26 Out MakeVectorAttribute_22 X)
  (connect CreateAttribute_24 Out MakeVectorAttribute_22 Y)
  (connect CreateAttribute_24 Out MakeVectorAttribute_22 Z)
  (connect ElevationIsolinesPoints Out Merge_16 In)
  (connect AttributeMathsOp_5 Out ResetPointCenter_10 In)
  (connect Branch_2 Output%20A SelfPruning_17 In)
  (connect ExtentsModifier_0 Out SelfPruning_17 In)
  (connect Union_0 Out Subgraph_1 SpatialData)
  (connect Intersection_2 Out ToPoint_4 In)
  (connect Intersection_17 Out ToPoint_19 In)
  (connect Branch_3 Output%20B Union_0 In)
  (output Points AttributeMathsOp_46))
