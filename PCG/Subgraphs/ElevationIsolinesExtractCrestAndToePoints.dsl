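; PCG subgraph that extracts crest and toe points from elevation isoline points and filters slope points by their distance to them.
; Projects points onto offset, rotated bounds sized from spacing inputs, comparing Z to classify crest, toe and slope.
; SlopeFilterMode selects the distance bands, union or inverse, merged with edge points, or passes all points.
; DisplaySlopeGradients diverts results to debug views showing distance as density, leaving Out empty.
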(graph ElevationIsolinesExtractCrestAndToePoints
  (params (SlopeFilterMode :type (Enum /PCG/Subgraphs/ElevationIsolineSlopeFilterMode.ElevationIsolineSlopeFilterMode)) (DistanceToCrest :type Double :default 5e+02) (DistanceToToe :type Double :default 5e+02) (SlopeFilterMaxDistance :type Double :default 1e+04) (DisplaySlopeGradients :type Bool))
  (inputs (ElevationLinesPointsIn :type Point) (MaxHorizontalDistance :type Param) (VerticalSpacing :type Param))
  (bind Switch_86 (Switch :EnumSelection (struct :Class /PCG/Subgraphs/ElevationIsolineSlopeFilterMode.ElevationIsolineSlopeFilterMode) :IntegerSelection 3 :IntOptions [0 1 2 3] :SelectionMode Enum :Value SlopeFilterMode))
  (bind Gather_91 (Gather))
  (bind CreateAttribute_8 (Create%20Constant :AttributeTypes (struct :DoubleValue 1.5)))
  (bind AttributeMathsOp_7 (Attribute%20Maths%20Op :Operation Multiply))
  (bind CreateAttribute_9 (Create%20Constant :AttributeTypes (struct :Type Vector)))
  (bind AttributeMathsOp_10 (Attribute%20Maths%20Op :InputSource1 @Last.z :Operation Set))
  (bind AttributeMathsOp_13 (Attribute%20Maths%20Op :Operation Multiply))
  (bind AttributeMathsOp_14 (Attribute%20Maths%20Op :InputSource1 @Last.y :Operation Set))
  (bind CreateAttribute_12 (Create%20Constant :AttributeTypes (struct :DoubleValue -1.0)))
  (bind AttributeMathsOp_15 (Attribute%20Maths%20Op :Operation Multiply))
  (bind MakeRotatorAttribute_1 (Make%20Rotator%20Attribute :InputSource1 $Rotation.Forward :Operation MakeRotFromX :OutputTarget $Rotation))
  (bind BoundsModifier_1 (Bounds%20Modifier :BoundsMax (0.0 0.0 375.0) :BoundsMin (0.0 -4.5e+02 0.0) :Mode Translate))
  (bind ExtentsModifier_0 (Extents%20Modifier :Extents (1.5 1.0 1.0) :Mode Multiply))
  (bind Projection_0 (Projection :bKeepZeroDensityPoints true :ProjectionParams (struct :bProjectRotations false)))
  (bind AttributeMathsOp_11 (Attribute%20Maths%20Op :Operation Multiply))
  (bind BoundsModifier_3 (Bounds%20Modifier :BoundsMax (0.0 4.5e+02 0.0) :BoundsMin (0.0 0.0 -375.0) :Mode Translate))
  (bind ExtentsModifier_2 (Extents%20Modifier :Extents (1.5 1.0 1.0) :Mode Multiply))
  (bind Projection_4 (Projection :bKeepZeroDensityPoints true :ProjectionParams (struct :bProjectRotations false)))
  (bind AttributeFilter_5 (Filter%20Attribute%20Elements :Operator Equal :TargetAttribute $Position.z :ThresholdAttribute $Position.z))
  (bind AttributeFilter_6 (Filter%20Attribute%20Elements :Operator Equal :TargetAttribute $Position.z :ThresholdAttribute $Position.z))
  (bind Inner%20Intersection_20 (Inner%20Intersection))
  (bind ToPoint_21 (To%20Point))
  (bind Difference_4 (Difference))
  (bind ToPoint_5 (To%20Point))
  (bind Switch_25 (Switch :EnumSelection (struct :Class /PCG/Subgraphs/ElevationIsolineSlopeFilterMode.ElevationIsolineSlopeFilterMode) :IntegerSelection 3 :IntOptions [0 1 2 3] :SelectionMode Enum :Value SlopeFilterMode))
  (bind Distance_0 (Distance :bOutputDistanceVector true :MaximumDistance 5e+03 :SourceShape Center :TargetShape Center :MaximumDistance-param SlopeFilterMaxDistance))
  (bind Distance_17 (Distance :bOutputDistanceVector true :MaximumDistance 5e+03 :SourceShape Center :TargetShape Center :MaximumDistance-param SlopeFilterMaxDistance))
  (bind AttributeFilter_22 (Filter%20Attribute%20Elements :bUseConstantThreshold true :bUseSpatialQuery true :Operator GreaterOrEqual :TargetAttribute Distance.z :ThresholdAttribute @Last))
  (bind CreateAttribute_3 (Create%20Constant))
  (bind AttributeFilterRange_1 (Filter%20Attribute%20Elements%20by%20Range :MaxThreshold (struct :ThresholdAttribute @Last :bUseSpatialQuery false) :MinThreshold (struct :bInclusive false :ThresholdAttribute @Last :bUseSpatialQuery false) :TargetAttribute Distance.length :FilterMax DistanceToCrest))
  (bind AttributeFilter_1 (Filter%20Attribute%20Elements :bUseConstantThreshold true :bUseSpatialQuery true :Operator LesserOrEqual :TargetAttribute Distance.z :ThresholdAttribute @Last))
  (bind AttributeFilterRange_6 (Filter%20Attribute%20Elements%20by%20Range :MaxThreshold (struct :ThresholdAttribute @Last :bUseSpatialQuery false) :MinThreshold (struct :bInclusive false :ThresholdAttribute @Last :bUseSpatialQuery false) :TargetAttribute Distance.length :FilterMax DistanceToToe))
  (bind Union_14 (Union))
  (bind ToPoint_16 (To%20Point))
  (bind Inner%20Intersection_13 (Inner%20Intersection))
  (bind ToPoint_18 (To%20Point))
  (bind Select%20%28Multi%29_1 (Select%20%28Multi%29 :EnumSelection (struct :Class /PCG/Subgraphs/ElevationIsolineSlopeFilterMode.ElevationIsolineSlopeFilterMode) :SelectionMode Enum :Value SlopeFilterMode))
  (bind Branch_44 (Branch :bOutputToB-param DistanceToCrest))
  (bind Branch_63 (Branch :bOutputToB-param DistanceToToe))
  (bind Select%20%28Multi%29_83 (Select%20%28Multi%29 :EnumSelection (struct :Class /PCG/Subgraphs/ElevationIsolineSlopeFilterMode.ElevationIsolineSlopeFilterMode) :SelectionMode Enum :Value SlopeFilterMode))
  (bind Merge_67 (Merge%20Points))
  (bind Branch_6 (Branch :bOutputToB true :bOutputToB-param DisplaySlopeGradients))
  (bind Debug_0 (Debug))
  (bind Debug_1 (Debug))
  ;[0 0 0] Debug colors
  (bind AttributeNoise_14 (Attribute%20Noise :InputSource $Color :OutputTarget $Color))
  ;[0 0 0] Debug Distance Gradients as Density
  (bind Branch_58 (Branch :bOutputToB true :bOutputToB-param DisplaySlopeGradients))
  (bind Switch_36 (Switch :EnumSelection (struct :Class /PCG/Subgraphs/ElevationIsolineSlopeFilterMode.ElevationIsolineSlopeFilterMode) :IntegerSelection 3 :IntOptions [0 1 2 3] :SelectionMode Enum :Value SlopeFilterMode))
  (bind Switch_54 (Switch :EnumSelection (struct :Class /PCG/Subgraphs/ElevationIsolineSlopeFilterMode.ElevationIsolineSlopeFilterMode) :IntegerSelection 3 :IntOptions [0 1 2 3] :SelectionMode Enum :Value SlopeFilterMode))
  (bind Distance_56 (Distance :bOutputDistanceVector true :MaximumDistance 5e+03 :SourceShape Center :TargetShape Center :MaximumDistance-param SlopeFilterMaxDistance))
  (bind AttributeMathsOp_5 (Attribute%20Maths%20Op :InputSource1 Distance.Length :Operation Divide :OutputTarget $Density :InB SlopeFilterMaxDistance))
  (bind AttributeReduce_24 (Attribute%20Reduce :InputSource $Density :Operation Max))
  (bind Select%20%28Multi%29_20 (Select%20%28Multi%29 :EnumSelection (struct :Class /PCG/Subgraphs/ElevationIsolineSlopeFilterMode.ElevationIsolineSlopeFilterMode) :SelectionMode Enum :Value SlopeFilterMode))
  (bind Branch_34 (Branch :bOutputToB true :bOutputToB-param DisplaySlopeGradients))
  (bind DensityRemap_21 (Density%20Remap :OutRangeMax 0.0))
  (bind Merge_22 (Merge%20Points))
  (bind DensityRemap_23 (Density%20Remap))
  (bind AttributeMathsOp_4 (Attribute%20Maths%20Op :InputSource1 $Density :Operation OneMinus))
  (bind Select%20%28Multi%29_6 (Select%20%28Multi%29 :EnumSelection (struct :Class /PCG/Subgraphs/ElevationIsolineSlopeFilterMode.ElevationIsolineSlopeFilterMode) :SelectionMode Enum :Value SlopeFilterMode))
  (bind ExtentsModifier_8 (Extents%20Modifier :Mode Add))
  (bind Debug_4 (Debug))
  (reroute CrestPoints AttributeFilter_6.InsideFilter)
  (reroute InputPoints ElevationLinesPointsIn)
  (reroute SlopePoints ToPoint_21)
  (reroute ToePoints ToPoint_5)
  (connect Distance_0 Out AttributeFilter_1 In)
  (connect Projection_0 Out AttributeFilter_5 Filter)
  (connect AttributeNoise_14 Out AttributeFilter_5 In)
  (connect Projection_4 Out AttributeFilter_6 Filter)
  (connect AttributeNoise_14 Out AttributeFilter_6 In)
  (connect Distance_17 Out AttributeFilter_22 In)
  (connect CreateAttribute_3 Out AttributeFilterRange_1 FilterMin)
  (connect AttributeFilter_22 InsideFilter AttributeFilterRange_1 In)
  (connect CreateAttribute_3 Out AttributeFilterRange_6 FilterMin)
  (connect AttributeFilter_1 InsideFilter AttributeFilterRange_6 In)
  (connect DensityRemap_23 Out AttributeMathsOp_4 In)
  (connect Distance_56 Out AttributeMathsOp_5 InA)
  (connect DefaultInputNode VerticalSpacing AttributeMathsOp_7 InA)
  (connect CreateAttribute_8 Out AttributeMathsOp_7 InB)
  (connect CreateAttribute_9 Out AttributeMathsOp_10 InA)
  (connect AttributeMathsOp_7 Out AttributeMathsOp_10 InB)
  (connect AttributeMathsOp_10 Out AttributeMathsOp_11 InA)
  (connect CreateAttribute_12 Out AttributeMathsOp_11 InB)
  (connect DefaultInputNode MaxHorizontalDistance AttributeMathsOp_13 InA)
  (connect CreateAttribute_8 Out AttributeMathsOp_13 InB)
  (connect CreateAttribute_9 Out AttributeMathsOp_14 InA)
  (connect AttributeMathsOp_13 Out AttributeMathsOp_14 InB)
  (connect AttributeMathsOp_14 Out AttributeMathsOp_15 InA)
  (connect CreateAttribute_12 Out AttributeMathsOp_15 InB)
  (connect Gather_91 Out AttributeNoise_14 In)
  (connect AttributeMathsOp_5 Out AttributeReduce_24 In)
  (connect AttributeMathsOp_10 Out BoundsModifier_1 BoundsMax)
  (connect AttributeMathsOp_15 Out BoundsModifier_1 BoundsMin)
  (connect MakeRotatorAttribute_1 Out BoundsModifier_1 In)
  (connect AttributeMathsOp_14 Out BoundsModifier_3 BoundsMax)
  (connect AttributeMathsOp_11 Out BoundsModifier_3 BoundsMin)
  (connect MakeRotatorAttribute_1 Out BoundsModifier_3 In)
  (connect Merge_67 Out Branch_6 In)
  (connect Select%20%28Multi%29_20 Out Branch_34 In)
  (connect CrestPoints Out Branch_44 In)
  (connect SlopePoints Out Branch_58 In)
  (connect ToePoints Out Branch_63 In)
  (connect Branch_6 Output%20B Debug_0 In)
  (connect ExtentsModifier_8 Out Debug_4 In)
  (connect Branch_34 Output%20B DensityRemap_21 In)
  (connect Merge_22 Out DensityRemap_23 In)
  (connect AttributeReduce_24 Out DensityRemap_23 InRangeMax)
  (connect AttributeFilter_6 InsideFilter Difference_4 Differences)
  (connect AttributeFilter_5 InsideFilter Difference_4 Source)
  (connect Switch_25 Distance%20To%20Toe Distance_0 Source)
  (connect Switch_25 Distance%20To%20Toe%20And%20Crest Distance_0 Source)
  (connect Switch_25 Inverse%20Distance%20To%20Toe%20And%20Crest Distance_0 Source)
  (connect ToePoints Out Distance_0 Target)
  (connect Switch_25 Distance%20To%20Crest Distance_17 Source)
  (connect Switch_25 Distance%20To%20Toe%20And%20Crest Distance_17 Source)
  (connect Switch_25 Inverse%20Distance%20To%20Toe%20And%20Crest Distance_17 Source)
  (connect CrestPoints Out Distance_17 Target)
  (connect Branch_58 Output%20B Distance_56 Source)
  (connect Switch_36 Distance%20To%20Crest Distance_56 Target)
  (connect Switch_36 Distance%20To%20Toe%20And%20Crest Distance_56 Target)
  (connect Switch_36 Inverse%20Distance%20To%20Toe%20And%20Crest Distance_56 Target)
  (connect Switch_54 Distance%20To%20Toe Distance_56 Target)
  (connect Switch_54 Distance%20To%20Toe%20And%20Crest Distance_56 Target)
  (connect Switch_54 Inverse%20Distance%20To%20Toe%20And%20Crest Distance_56 Target)
  (connect BoundsModifier_1 Out ExtentsModifier_0 In)
  (connect BoundsModifier_3 Out ExtentsModifier_2 In)
  (connect Select%20%28Multi%29_6 Out ExtentsModifier_8 In)
  (connect Switch_86 Distance%20To%20Crest Gather_91 In)
  (connect Switch_86 Distance%20To%20Toe Gather_91 In)
  (connect Switch_86 Distance%20To%20Toe%20And%20Crest Gather_91 In)
  (connect Switch_86 Inverse%20Distance%20To%20Toe%20And%20Crest Gather_91 In)
  (connect AttributeFilterRange_1 OutsideFilter Inner%20Intersection_13 In)
  (connect AttributeFilterRange_6 OutsideFilter Inner%20Intersection_13 In)
  (connect AttributeFilter_5 OutsideFilter Inner%20Intersection_20 In)
  (connect AttributeFilter_6 OutsideFilter Inner%20Intersection_20 In)
  (connect Gather_91 Out MakeRotatorAttribute_1 X)
  (connect AttributeMathsOp_5 Out Merge_22 In)
  (connect DensityRemap_21 Out Merge_22 In)
  (connect Select%20%28Multi%29_1 Out Merge_67 In)
  (connect Select%20%28Multi%29_83 Out Merge_67 In2)
  (connect AttributeNoise_14 Out Projection_0 In)
  (connect ExtentsModifier_0 Out Projection_0 Projection%20Target)
  (connect AttributeNoise_14 Out Projection_4 In)
  (connect ExtentsModifier_2 Out Projection_4 Projection%20Target)
  (connect AttributeFilterRange_1 InsideFilter Select%20%28Multi%29_1 Distance%20To%20Crest)
  (connect AttributeFilterRange_6 InsideFilter Select%20%28Multi%29_1 Distance%20To%20Toe)
  (connect ToPoint_16 Out Select%20%28Multi%29_1 Distance%20To%20Toe%20And%20Crest)
  (connect ToPoint_18 Out Select%20%28Multi%29_1 Inverse%20Distance%20To%20Toe%20And%20Crest)
  (connect AttributeMathsOp_4 Out Select%20%28Multi%29_6 Distance%20To%20Crest)
  (connect AttributeMathsOp_4 Out Select%20%28Multi%29_6 Distance%20To%20Toe)
  (connect AttributeMathsOp_4 Out Select%20%28Multi%29_6 Distance%20To%20Toe%20And%20Crest)
  (connect DensityRemap_23 Out Select%20%28Multi%29_6 Inverse%20Distance%20To%20Toe%20And%20Crest)
  (connect CrestPoints Out Select%20%28Multi%29_20 Distance%20To%20Crest)
  (connect ToePoints Out Select%20%28Multi%29_20 Distance%20To%20Toe)
  (connect ToePoints Out Select%20%28Multi%29_20 Distance%20To%20Toe%20And%20Crest)
  (connect CrestPoints Out Select%20%28Multi%29_20 Distance%20To%20Toe%20And%20Crest)
  (connect ToePoints Out Select%20%28Multi%29_20 Inverse%20Distance%20To%20Toe%20And%20Crest)
  (connect CrestPoints Out Select%20%28Multi%29_20 Inverse%20Distance%20To%20Toe%20And%20Crest)
  (connect InputPoints Out Select%20%28Multi%29_83 All%20Points)
  (connect CrestPoints Out Select%20%28Multi%29_83 Distance%20To%20Crest)
  (connect ToePoints Out Select%20%28Multi%29_83 Distance%20To%20Toe)
  (connect ToePoints Out Select%20%28Multi%29_83 Distance%20To%20Toe%20And%20Crest)
  (connect CrestPoints Out Select%20%28Multi%29_83 Distance%20To%20Toe%20And%20Crest)
  (connect Branch_44 Output%20A Select%20%28Multi%29_83 Inverse%20Distance%20To%20Toe%20And%20Crest)
  (connect Branch_63 Output%20A Select%20%28Multi%29_83 Inverse%20Distance%20To%20Toe%20And%20Crest)
  (connect SlopePoints Out Switch_25 In)
  (connect CrestPoints Out Switch_36 In)
  (connect ToePoints Out Switch_54 In)
  (connect InputPoints Out Switch_86 In)
  (connect Difference_4 Out ToPoint_5 In)
  (connect Union_14 Out ToPoint_16 In)
  (connect Inner%20Intersection_13 Out ToPoint_18 In)
  (connect Inner%20Intersection_20 Out ToPoint_21 In)
  (connect AttributeFilterRange_1 InsideFilter Union_14 In)
  (connect AttributeFilterRange_6 InsideFilter Union_14 In)
  (output Out Branch_6.Output%20A))
