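; PCG subgraph that filters elevation isoline points from the current loop by spacing, corners and isolation.
; It flattens Z to measure 2D distance to other lines' points, then restores it and keeps a min-max distance band.
; Corners are rejected by dotting right vectors against rotations projected from neighbors.
; Points barely displaced by neighborhood position projection are dropped as isolated, then density and tags are reset, with optional debug display.
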(graph ElevationIsolinesFilters
  (params (MinHorizontalDistance :type Double :default 6e+01) (MaxHorizontalDistance :type Double :default 4e+02) (CornerFilterSearchDistance :type Double :default 1.5e+03) (NeighborSearchDistance :type Double :default 2.5e+02) (DisplayPoints :type Bool))
  (inputs (PointDataLoop :type Point) (AllPointData :type Point))
  ;[0 0 0] Filter line points using 2d distance from other lines points
  (bind CreateAttribute_3 (Create%20Constant :AttributeTypes (struct :DoubleValue 1.0)))
  (bind AttributeMathsOp_2 (Attribute%20Maths%20Op :InA MaxHorizontalDistance))
  (bind CopyAttribute_27 (Copy%20Attributes :InputSource $Position.z :OutputTarget z))
  (bind CreateAttribute_29 (Create%20Constant))
  (bind AttributeMathsOp_28 (Attribute%20Maths%20Op :InputSource1 $Position.z :Operation Set))
  (bind Difference_2 (Difference :bDiffMetadata false :DensityFunction Binary))
  (bind ToPoint_3 (To%20Point))
  (bind AttributeMathsOp_31 (Attribute%20Maths%20Op :InputSource1 $Position.z :Operation Set))
  (bind Distance_1 (Distance :bSetDensity true :MaximumDistance 2.5e+04 :OutputAttribute 2DDist :SourceShape Center :TargetShape BoxBounds))
  (bind CreateAttribute_14 (Create%20Constant))
  (bind AttributeMathsOp_15 (Attribute%20Maths%20Op :Operation Max))
  (bind CopyAttribute_33 (Copy%20Attributes :InputSource z :OutputTarget $Position.z))
  (bind DeleteAttributes_34 (Delete%20Attributes :bTokenizeOnWhiteSpace true :SelectedAttributes "z"))
  ;[0 0 0] Remove points in tight corners, make sure all points are pointing up
  (bind CopyAttribute_4 (Copy%20Attributes :InputSource $Rotation :OutputTarget rot))
  (bind MakeRotatorAttribute_0 (Make%20Rotator%20Attribute :InputSource1 $Rotation.Forward :Operation MakeRotFromX :OutputTarget $Rotation))
  (bind MakeVectorAttribute_9 (Make%20Vector%20Attribute :OutputType Vector :X CornerFilterSearchDistance :Y CornerFilterSearchDistance))
  (bind ExtentsModifier_3 (Extents%20Modifier :Mode Add))
  (bind Projection_2 (Projection :ProjectionParams (struct :bProjectPositions false)))
  (bind AttributeVectorOp_5 (Attribute%20Vector%20Op :InputSource1 $Rotation.right :InputSource2 $Rotation.right :Operation Dot :OutputTarget $Density))
  (bind CopyAttribute_5 (Copy%20Attributes :InputSource rot :OutputTarget $Rotation))
  (bind DeleteAttributes_6 (Delete%20Attributes :bTokenizeOnWhiteSpace true :SelectedAttributes "rot"))
  ;[0 0 0] Cleanup output
  (bind GetPointsCount_0 (Get%20Element%20Count))
  (bind Branch_1 (Branch))
  (bind CreateAttribute_20 (Create%20Constant :AttributeTypes (struct :DoubleValue 1.0)))
  (bind AttributeMathsOp_19 (Attribute%20Maths%20Op :InputSource1 $Density :Operation Set))
  (bind DeleteTags_17 (Delete%20Tags :bTokenizeOnWhiteSpace true :Operation KeepOnlySelectedTags))
  (bind Branch_25 (Branch :bOutputToB true :bOutputToB-param DisplayPoints))
  (bind Debug_26 (Debug))
  ;[0 0 0] Remove Isolated Points
  (bind MakeVectorAttribute_16 (Make%20Vector%20Attribute :OutputType Vector :X NeighborSearchDistance :Y NeighborSearchDistance))
  (bind ExtentsModifier_2 (Extents%20Modifier :Extents (1.5e+02 1.5e+02 0.0) :Mode Add))
  (bind Projection_0 (Projection :ProjectionParams (struct :bProjectRotations false)))
  (bind Distance_7 (Distance :bOutputToAttribute false :bSetDensity true :MaximumDistance 1e+01 :OutputAttribute DistFromAvg :SourceShape Center :TargetShape Center))
  (bind DensityFilter_10 (Density%20Filter :LowerBound 0.1))
  ;[0 0 0] Apply 2D Distance and Corner filtering
  (bind AttributeFilterRange_26 (Filter%20Attribute%20Elements%20by%20Range :MaxThreshold (struct :ThresholdAttribute @Last :bUseSpatialQuery false) :MinThreshold (struct :ThresholdAttribute @Last :bUseSpatialQuery false) :TargetAttribute 2DDist :FilterMax MaxHorizontalDistance :FilterMin MinHorizontalDistance))
  (bind DensityFilter_50 (Density%20Filter :LowerBound 0.9))
  (bind DeleteAttributes_42 (Delete%20Attributes :bTokenizeOnWhiteSpace true :SelectedAttributes "2DDist"))
  (reroute 2D%20Dist%20Filter%20Out DeleteAttributes_34)
  (reroute CornerFilterOut DeleteAttributes_6)
  (reroute FilteredPoints DeleteAttributes_42)
  (reroute Input%20Node%20AllPointData AllPointData)
  (reroute Input%20Node%20PointDataLoop PointDataLoop)
  (reroute IsolatedPointsFilterOut DensityFilter_10)
  (connect CornerFilterOut Out AttributeFilterRange_26 In)
  (connect CreateAttribute_3 Out AttributeMathsOp_2 InB)
  (connect Distance_1 Out AttributeMathsOp_15 InA)
  (connect CreateAttribute_14 Out AttributeMathsOp_15 InB)
  (connect Branch_1 Output%20B AttributeMathsOp_19 InA)
  (connect CreateAttribute_20 Out AttributeMathsOp_19 InB)
  (connect CopyAttribute_27 Out AttributeMathsOp_28 InA)
  (connect CreateAttribute_29 Out AttributeMathsOp_28 InB)
  (connect ToPoint_3 Out AttributeMathsOp_31 InA)
  (connect CreateAttribute_29 Out AttributeMathsOp_31 InB)
  (connect MakeRotatorAttribute_0 Out AttributeVectorOp_5 InA)
  (connect Projection_2 Out AttributeVectorOp_5 InB)
  (connect GetPointsCount_0 Out Branch_1 bOutputToB)
  (connect IsolatedPointsFilterOut Out Branch_1 In)
  (connect DeleteTags_17 Out Branch_25 In)
  (connect 2D%20Dist%20Filter%20Out Out CopyAttribute_4 Source)
  (connect 2D%20Dist%20Filter%20Out Out CopyAttribute_4 Target)
  (connect AttributeVectorOp_5 Out CopyAttribute_5 Source)
  (connect AttributeVectorOp_5 Out CopyAttribute_5 Target)
  (connect Input%20Node%20PointDataLoop Out CopyAttribute_27 Source)
  (connect Input%20Node%20PointDataLoop Out CopyAttribute_27 Target)
  (connect AttributeMathsOp_15 Out CopyAttribute_33 Source)
  (connect AttributeMathsOp_15 Out CopyAttribute_33 Target)
  (connect Branch_25 Output%20B Debug_26 In)
  (connect CopyAttribute_5 Out DeleteAttributes_6 In)
  (connect CopyAttribute_33 Out DeleteAttributes_34 In)
  (connect DensityFilter_50 Out DeleteAttributes_42 In)
  (connect AttributeMathsOp_19 Out DeleteTags_17 In)
  (connect Distance_7 Out DensityFilter_10 In)
  (connect AttributeFilterRange_26 InsideFilter DensityFilter_50 In)
  (connect Input%20Node%20PointDataLoop Out Difference_2 Differences)
  (connect Input%20Node%20AllPointData Out Difference_2 Source)
  (connect AttributeMathsOp_2 Out Distance_1 MaximumDistance)
  (connect AttributeMathsOp_28 Out Distance_1 Source)
  (connect AttributeMathsOp_31 Out Distance_1 Target)
  (connect FilteredPoints Out Distance_7 Source)
  (connect Projection_0 Out Distance_7 Target)
  (connect MakeVectorAttribute_16 Out ExtentsModifier_2 Extents)
  (connect FilteredPoints Out ExtentsModifier_2 In)
  (connect MakeVectorAttribute_9 Out ExtentsModifier_3 Extents)
  (connect MakeRotatorAttribute_0 Out ExtentsModifier_3 In)
  (connect IsolatedPointsFilterOut Out GetPointsCount_0 In)
  (connect CopyAttribute_4 Out MakeRotatorAttribute_0 X)
  (connect FilteredPoints Out Projection_0 In)
  (connect ExtentsModifier_2 Out Projection_0 Projection%20Target)
  (connect MakeRotatorAttribute_0 Out Projection_2 In)
  (connect ExtentsModifier_3 Out Projection_2 Projection%20Target)
  (connect Difference_2 Out ToPoint_3 In)
  (output Out DeleteTags_17))
