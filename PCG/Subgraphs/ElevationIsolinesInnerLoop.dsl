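; PCG subgraph that generates points along elevation isolines where world geometry meets stacked height slices of a single input point's oriented bounds.
; Horizontal splines, built from subdivided vertical corner edges, are projected onto world ray-hit queries from four sides, then sampled.
; Optional experimental resampling projects points and cuts rebuilt splines via CutSplines.
; Grazing-angle points are culled, optionally color-coded by direction, then reoriented using the Surface input, with optional Z-up.
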(graph ElevationIsolinesInnerLoop
  (params (SplineSamplingDistance :type Double :default 2.5e+02) (lineCount :type Int :default 10) (ZUp :type Bool :default true) (IgnorePCGHits :type Bool :default true) (CollisionChannel :type (Enum /Script/Engine.ECollisionChannel)) (SetSplineColors :type Bool) (ExperimentalResampling :type Bool) (ExperimentalResamplingCutSplinesDistThreshold :type Double :default 2.5e+02) (ExperimentalResamplingInternalDistance :type Double :default 1e+02))
  (inputs (InPoint :type Point) (Surface :type Surface))
  (outputs (OutPoints :type Point))
  (bind AttributeMathsOp_74 (Attribute%20Maths%20Op :InputSource1 $Density :Operation Abs))
  (bind DensityFilter_24 (Density%20Filter :LowerBound 0.7))
  (bind Branch_27 (Branch :bOutputToB-param SetSplineColors))
  (bind AddTags_118 (Add%20Tags :bTokenizeOnWhiteSpace true :TagsToAdd "xp"))
  (bind Branch_0 (Branch :bOutputToB-param ExperimentalResampling))
  (bind AddTags_120 (Add%20Tags :bTokenizeOnWhiteSpace true :TagsToAdd "xn"))
  (bind Branch_2 (Branch :bOutputToB-param ExperimentalResampling))
  (bind AddTags_122 (Add%20Tags :bTokenizeOnWhiteSpace true :TagsToAdd "yp"))
  (bind Branch_4 (Branch :bOutputToB-param ExperimentalResampling))
  (bind AddTags_124 (Add%20Tags :bTokenizeOnWhiteSpace true :TagsToAdd "yn"))
  (bind Branch_6 (Branch :bOutputToB-param ExperimentalResampling))
  (bind FilterByType_8 (Filter%20Data%20By%20Type :TargetType (struct :Ids [(struct :Struct /Script/PCG.PCGDataTypeInfoPolyline)])))
  (bind SplineSampler_11 (Spline%20Sampler :SamplerParams (struct :Mode Distance :DistanceIncrement 2e+02 :bUnbounded true :PointSteepness 1.0) :DistanceIncrement SplineSamplingDistance))
  (bind ToPoint_10 (To%20Point))
  (bind CreateSpline_19 (Create%20Spline :ArriveTangentAttribute None :InterpTypeAttribute None :LeaveTangentAttribute None))
  (bind Loop_112 (Loop :bTokenizeOnWhiteSpace true :Graph /PCG/Subgraphs/CutSplines.CutSplines :DistanceThreshold ExperimentalResamplingCutSplinesDistThreshold))
  (bind SplineSampler_107 (Spline%20Sampler :SamplerParams (struct :Mode Distance :DistanceIncrement 5e+02 :bUnbounded true) :DistanceIncrement SplineSamplingDistance))
  (bind Gather_11 (Gather))
  ;[0 0 0] Make sure we have a single point here
  (bind CreateAttribute_5 (Create%20Constant :AttributeTypes (struct :Type Integer32 :Int32Value 1)))
  (bind GetPointsCount_3 (Get%20Element%20Count))
  (bind AttributeCompareOp_4 (Attribute%20Compare%20Op))
  (bind Branch_1 (Branch))
  ;[0 0 0] Reset point Scale - Apply to bounds
  (bind ApplyScaleToBounds_0 (Apply%20Scale%20To%20Bounds))
  ;[0 0 0] Ray Lengths from Input Point Bounds
  (bind GetAttributeFromPointIndex_20 (Get%20Attribute%20From%20Point%20Index :bForceOutputAttributeToBeInElementsDomain true :InputSource $BoundsMax :OutputAttributeName None))
  (bind GetAttributeFromPointIndex_22 (Get%20Attribute%20From%20Point%20Index :bForceOutputAttributeToBeInElementsDomain true :InputSource $BoundsMin :OutputAttributeName None))
  (bind AttributeMathsOp_30 (Attribute%20Maths%20Op :Operation Subtract))
  (bind BreakVectorAttribute_31 (Break%20Vector%20Attribute))
  ;[0 0 0] Ray Directions from Input Point Rotation
  (bind GetAttributeFromPointIndex_27 (Get%20Attribute%20From%20Point%20Index :bForceOutputAttributeToBeInElementsDomain true :InputSource $Rotation.right :OutputAttributeName None))
  (bind CreateAttribute_37 (Create%20Constant :AttributeTypes (struct :DoubleValue -1.0)))
  (bind AttributeMathsOp_36 (Attribute%20Maths%20Op :Operation Multiply))
  (bind GetAttributeFromPointIndex_70 (Get%20Attribute%20From%20Point%20Index :bForceOutputAttributeToBeInElementsDomain true :InputSource $Rotation.forward :OutputAttributeName None))
  (bind CreateAttribute_89 (Create%20Constant :AttributeTypes (struct :DoubleValue -1.0)))
  (bind AttributeMathsOp_87 (Attribute%20Maths%20Op :Operation Multiply))
  ;[0 0 0] Ray Origins at Input Point face Centers
  (bind MakeVectorAttribute_49 (Make%20Vector%20Attribute :InputSource1 $LocalCenter.X :InputSource2 $BoundsMin.Y :InputSource3 $LocalCenter.Z :InputSource4 None :OutputTarget FaceCenter :OutputType Vector))
  (bind AttributeVectorOp_45 (Attribute%20Vector%20Op :InputSource1 FaceCenter :InputSource2 $Transform :Operation TransformLocation))
  (bind GetAttributeFromPointIndex_47 (Get%20Attribute%20From%20Point%20Index :bForceOutputAttributeToBeInElementsDomain true :OutputAttributeName None))
  (bind MakeVectorAttribute_61 (Make%20Vector%20Attribute :InputSource1 $LocalCenter.X :InputSource2 $BoundsMax.Y :InputSource3 $LocalCenter.Z :InputSource4 None :OutputTarget FaceCenter :OutputType Vector))
  (bind AttributeVectorOp_57 (Attribute%20Vector%20Op :InputSource1 FaceCenter :InputSource2 $Transform :Operation TransformLocation))
  (bind GetAttributeFromPointIndex_59 (Get%20Attribute%20From%20Point%20Index :bForceOutputAttributeToBeInElementsDomain true :OutputAttributeName None))
  (bind MakeVectorAttribute_37 (Make%20Vector%20Attribute :InputSource1 $BoundsMin.X :InputSource2 $LocalCenter.Y :InputSource3 $LocalCenter.Z :InputSource4 None :OutputTarget FaceCenter :OutputType Vector))
  (bind AttributeVectorOp_18 (Attribute%20Vector%20Op :InputSource1 FaceCenter :InputSource2 $Transform :Operation TransformLocation))
  (bind GetAttributeFromPointIndex_29 (Get%20Attribute%20From%20Point%20Index :bForceOutputAttributeToBeInElementsDomain true :OutputAttributeName None))
  (bind MakeVectorAttribute_43 (Make%20Vector%20Attribute :InputSource1 $BoundsMax.X :InputSource2 $LocalCenter.Y :InputSource3 $LocalCenter.Z :InputSource4 None :OutputTarget FaceCenter :OutputType Vector))
  (bind AttributeVectorOp_39 (Attribute%20Vector%20Op :InputSource1 FaceCenter :InputSource2 $Transform :Operation TransformLocation))
  (bind GetAttributeFromPointIndex_41 (Get%20Attribute%20From%20Point%20Index :bForceOutputAttributeToBeInElementsDomain true :OutputAttributeName None))
  ;[0 0 0] RayCast The world Horizontally in 4 directions
  (bind WorldRayHitQuery_12 (World%20Ray%20Hit%20Query :QueryParams (struct :bOverrideDefaultParams true :RayDirection (0.0 1.0 0.0) :CollisionChannel ECC_Visibility) :bIgnorePCGHits IgnorePCGHits :CollisionChannel CollisionChannel))
  (bind WorldRayHitQuery_35 (World%20Ray%20Hit%20Query :QueryParams (struct :bOverrideDefaultParams true :RayDirection (0.0 1.0 0.0) :CollisionChannel ECC_Visibility) :bIgnorePCGHits IgnorePCGHits :CollisionChannel CollisionChannel))
  (bind WorldRayHitQuery_66 (World%20Ray%20Hit%20Query :QueryParams (struct :bOverrideDefaultParams true :RayDirection (0.0 1.0 0.0)) :bIgnorePCGHits IgnorePCGHits :CollisionChannel CollisionChannel))
  (bind WorldRayHitQuery_82 (World%20Ray%20Hit%20Query :QueryParams (struct :bOverrideDefaultParams true :RayDirection (0.0 1.0 0.0) :CollisionChannel ECC_Visibility) :bIgnorePCGHits IgnorePCGHits :CollisionChannel CollisionChannel))
  ;[0 0 0] Vertical Splines Start/End points
  (bind MakeVectorAttribute_69 (Make%20Vector%20Attribute :InputSource1 $BoundsMin.X :InputSource2 $BoundsMin.Y :InputSource3 $BoundsMin.Z :InputSource4 None :OutputTarget bMin :OutputType Vector))
  (bind AttributeVectorOp_71 (Attribute%20Vector%20Op :InputSource2 $Transform :Operation TransformLocation :OutputTarget $Position))
  (bind MakeVectorAttribute_75 (Make%20Vector%20Attribute :InputSource1 $BoundsMin.X :InputSource2 $BoundsMin.Y :InputSource3 $BoundsMax.Z :InputSource4 None :OutputTarget bMin :OutputType Vector))
  (bind AttributeVectorOp_77 (Attribute%20Vector%20Op :InputSource2 $Transform :Operation TransformLocation :OutputTarget $Position))
  (bind MakeVectorAttribute_95 (Make%20Vector%20Attribute :InputSource1 $BoundsMax.X :InputSource2 $BoundsMin.Y :InputSource3 $BoundsMin.Z :InputSource4 None :OutputTarget bMin :OutputType Vector))
  (bind AttributeVectorOp_97 (Attribute%20Vector%20Op :InputSource2 $Transform :Operation TransformLocation :OutputTarget $Position))
  (bind MakeVectorAttribute_99 (Make%20Vector%20Attribute :InputSource1 $BoundsMax.X :InputSource2 $BoundsMin.Y :InputSource3 $BoundsMax.Z :InputSource4 None :OutputTarget bMin :OutputType Vector))
  (bind AttributeVectorOp_101 (Attribute%20Vector%20Op :InputSource2 $Transform :Operation TransformLocation :OutputTarget $Position))
  (bind MakeVectorAttribute_83 (Make%20Vector%20Attribute :InputSource1 $BoundsMin.X :InputSource2 $BoundsMax.Y :InputSource3 $BoundsMin.Z :InputSource4 None :OutputTarget bMin :OutputType Vector))
  (bind AttributeVectorOp_85 (Attribute%20Vector%20Op :InputSource2 $Transform :Operation TransformLocation :OutputTarget $Position))
  (bind MakeVectorAttribute_87 (Make%20Vector%20Attribute :InputSource1 $BoundsMin.X :InputSource2 $BoundsMax.Y :InputSource3 $BoundsMax.Z :InputSource4 None :OutputTarget bMin :OutputType Vector))
  (bind AttributeVectorOp_89 (Attribute%20Vector%20Op :InputSource2 $Transform :Operation TransformLocation :OutputTarget $Position))
  ;[0 0 0] Build Vertical Splines
  (bind Merge_79 (Merge%20Points))
  (bind CreateSpline_81 (Create%20Spline :ArriveTangentAttribute None :bLinear true :InterpTypeAttribute None :LeaveTangentAttribute None))
  (bind SplineSampler_27 (Spline%20Sampler :SamplerParams (struct :SubdivisionsPerSegment 10 :DistanceIncrement 1e+03 :bUnbounded true :PointSteepness 1.0) :SubdivisionsPerSegment lineCount))
  (bind Merge_103 (Merge%20Points))
  (bind CreateSpline_105 (Create%20Spline :ArriveTangentAttribute None :bLinear true :InterpTypeAttribute None :LeaveTangentAttribute None))
  (bind SplineSampler_67 (Spline%20Sampler :SamplerParams (struct :SubdivisionsPerSegment 10 :DistanceIncrement 1e+03 :bUnbounded true :PointSteepness 1.0) :SubdivisionsPerSegment lineCount))
  (bind Merge_91 (Merge%20Points))
  (bind CreateSpline_93 (Create%20Spline :ArriveTangentAttribute None :bLinear true :InterpTypeAttribute None :LeaveTangentAttribute None))
  (bind SplineSampler_41 (Spline%20Sampler :SamplerParams (struct :SubdivisionsPerSegment 10 :DistanceIncrement 1e+03 :bUnbounded true :PointSteepness 1.0) :SubdivisionsPerSegment lineCount))
  ;[0 0 0] Build Pairs on Start/End Points by Partitioning points using z Pos
  (bind Merge_69 (Merge%20Points))
  (bind AttributePartition_73 (Attribute%20Partition :bTokenizeOnWhiteSpace true :PartitionAttributeSelectors [$Position.z]))
  (bind Merge_42 (Merge%20Points))
  (bind AttributePartition_46 (Attribute%20Partition :bTokenizeOnWhiteSpace true :PartitionAttributeSelectors [$Position.z]))
  ;[0 0 0] Build Horizontal Splines
  (bind CreateSpline_75 (Create%20Spline :ArriveTangentAttribute None :bLinear true :InterpTypeAttribute None :LeaveTangentAttribute None))
  (bind CreateSpline_48 (Create%20Spline :ArriveTangentAttribute None :bLinear true :InterpTypeAttribute None :LeaveTangentAttribute None))
  ;[0 0 0] Project splines directly or points to create and sample new splines
  (bind Branch_14 (Branch :bOutputToB-param ExperimentalResampling))
  (bind SplineSampler_25 (Spline%20Sampler :SamplerParams (struct :Mode Distance :SubdivisionsPerSegment 0 :PointSteepness 1.0) :DistanceIncrement ExperimentalResamplingInternalDistance))
  (bind Gather_21 (Gather))
  (bind Branch_16 (Branch :bOutputToB-param ExperimentalResampling))
  (bind SplineSampler_23 (Spline%20Sampler :SamplerParams (struct :Mode Distance :SubdivisionsPerSegment 0 :PointSteepness 1.0) :DistanceIncrement ExperimentalResamplingInternalDistance))
  (bind Gather_19 (Gather))
  ;[0 0 0] Project Horizontal Splines on Physics World
  (bind Projection_13 (Projection))
  (bind Projection_42 (Projection))
  (bind Projection_72 (Projection))
  (bind Projection_91 (Projection))
  ;[0 0 0] Filter out points at grazing angles
  (bind CreateAttribute_23 (Create%20Constant :AttributeTypes (struct :Type Vector :VectorValue (1.0 0.0 0.0))))
  (bind FilterDataByTag_25 (Filter%20Data%20By%20Tag :bTokenizeOnWhiteSpace true :SelectedTags "xp"))
  (bind MakeRotatorAttribute_21 (Make%20Rotator%20Attribute :InputSource1 $Rotation.Up :Operation MakeRotFromZ :OutputTarget $Rotation))
  (bind MakeRotatorAttribute_19 (Make%20Rotator%20Attribute :InputSource1 $Rotation.Forward :Operation MakeRotFromX :OutputTarget $Rotation))
  (bind AttributeVectorOp_22 (Attribute%20Vector%20Op :InputSource1 $Rotation.right :Operation Dot :OutputTarget $Density))
  (bind CreateAttribute_35 (Create%20Constant :AttributeTypes (struct :Type Vector :VectorValue (-1.0 0.0 0.0))))
  (bind FilterDataByTag_39 (Filter%20Data%20By%20Tag :bTokenizeOnWhiteSpace true :SelectedTags "xn"))
  (bind MakeRotatorAttribute_31 (Make%20Rotator%20Attribute :InputSource1 $Rotation.Up :Operation MakeRotFromZ :OutputTarget $Rotation))
  (bind MakeRotatorAttribute_29 (Make%20Rotator%20Attribute :InputSource1 $Rotation.Forward :Operation MakeRotFromX :OutputTarget $Rotation))
  (bind AttributeVectorOp_33 (Attribute%20Vector%20Op :InputSource1 $Rotation.right :Operation Dot :OutputTarget $Density))
  (bind CreateAttribute_49 (Create%20Constant :AttributeTypes (struct :Type Vector :VectorValue (0.0 1.0 0.0))))
  (bind FilterDataByTag_53 (Filter%20Data%20By%20Tag :bTokenizeOnWhiteSpace true :SelectedTags "yp"))
  (bind MakeRotatorAttribute_45 (Make%20Rotator%20Attribute :InputSource1 $Rotation.Up :Operation MakeRotFromZ :OutputTarget $Rotation))
  (bind MakeRotatorAttribute_43 (Make%20Rotator%20Attribute :InputSource1 $Rotation.Forward :Operation MakeRotFromX :OutputTarget $Rotation))
  (bind AttributeVectorOp_47 (Attribute%20Vector%20Op :InputSource1 $Rotation.right :Operation Dot :OutputTarget $Density))
  (bind CreateAttribute_63 (Create%20Constant :AttributeTypes (struct :Type Vector :VectorValue (0.0 -1.0 0.0))))
  (bind FilterDataByTag_67 (Filter%20Data%20By%20Tag :bTokenizeOnWhiteSpace true :SelectedTags "yn"))
  (bind MakeRotatorAttribute_59 (Make%20Rotator%20Attribute :InputSource1 $Rotation.Up :Operation MakeRotFromZ :OutputTarget $Rotation))
  (bind MakeRotatorAttribute_57 (Make%20Rotator%20Attribute :InputSource1 $Rotation.Forward :Operation MakeRotFromX :OutputTarget $Rotation))
  (bind AttributeVectorOp_61 (Attribute%20Vector%20Op :InputSource1 $Rotation.right :Operation Dot :OutputTarget $Density))
  ;[0 0 0] Debug Colors
  (bind CreateAttribute_137 (Create%20Constant :AttributeTypes (struct :Type Vector4)))
  (bind FilterByTag_125 (Filter%20Data%20By%20Tag :bTokenizeOnWhiteSpace true :SelectedTags "xp"))
  (bind CopyAttribute_136 (Copy%20Attributes :InputSource @Last :OutputTarget $Color))
  (bind CreateAttribute_141 (Create%20Constant :AttributeTypes (struct :Type Vector4 :Vector4Value (1.0 0.0 0.0 0.0))))
  (bind FilterByTag_128 (Filter%20Data%20By%20Tag :bTokenizeOnWhiteSpace true :SelectedTags "xn"))
  (bind CopyAttribute_139 (Copy%20Attributes :InputSource @Last :OutputTarget $Color))
  (bind CreateAttribute_145 (Create%20Constant :AttributeTypes (struct :Type Vector4 :Vector4Value (0.0 1.0 0.0 0.0))))
  (bind FilterByTag_132 (Filter%20Data%20By%20Tag :bTokenizeOnWhiteSpace true :SelectedTags "yp"))
  (bind CopyAttribute_143 (Copy%20Attributes :InputSource @Last :OutputTarget $Color))
  (bind CreateAttribute_149 (Create%20Constant :AttributeTypes (struct :Type Vector4 :Vector4Value (0.0 0.0 1.0 0.0))))
  (bind FilterByTag_134 (Filter%20Data%20By%20Tag :bTokenizeOnWhiteSpace true :SelectedTags "yn"))
  (bind CopyAttribute_147 (Copy%20Attributes :InputSource @Last :OutputTarget $Color))
  (bind Gather_83 (Gather))
  ;[0 0 0] Project Rotation from landscape to the points, rebuild rotations
  (bind Merge_46 (Merge%20Points :bMergeMetadata false))
  (bind Projection_48 (Projection :ProjectionParams (struct :bProjectPositions false)))
  (bind ToPoint_49 (To%20Point))
  (bind MakeRotatorAttribute_50 (Make%20Rotator%20Attribute :InputSource1 $Rotation.Up :Operation MakeRotFromZ :OutputTarget $Rotation))
  (bind Branch_30 (Branch :bOutputToB true :bOutputToB-param ZUp))
  (bind MakeRotatorAttribute_55 (Make%20Rotator%20Attribute :InputSource1 $Rotation.Forward :Operation MakeRotFromX :OutputTarget $Rotation))
  (bind TransformPoints_56 (Transform%20Points :RotationMax (struct :Yaw 1.8e+02) :RotationMin (struct :Yaw 1.8e+02)))
  (reroute Input%20Surface Surface)
  (reroute InputPoint ApplyScaleToBounds_0)
  (reroute Sampled%20Spline%20Points Gather_11)
  (connect Projection_72 Out AddTags_118 In)
  (connect Projection_91 Out AddTags_120 In)
  (connect Projection_13 Out AddTags_122 In)
  (connect Projection_42 Out AddTags_124 In)
  (connect Branch_1 Output%20B ApplyScaleToBounds_0 In)
  (connect GetPointsCount_3 Out AttributeCompareOp_4 InA)
  (connect CreateAttribute_5 Out AttributeCompareOp_4 InB)
  (connect GetAttributeFromPointIndex_20 Attribute AttributeMathsOp_30 InA)
  (connect GetAttributeFromPointIndex_22 Attribute AttributeMathsOp_30 InB)
  (connect GetAttributeFromPointIndex_27 Attribute AttributeMathsOp_36 InA)
  (connect CreateAttribute_37 Out AttributeMathsOp_36 InB)
  (connect AttributeVectorOp_22 Out AttributeMathsOp_74 In)
  (connect AttributeVectorOp_33 Out AttributeMathsOp_74 In)
  (connect AttributeVectorOp_47 Out AttributeMathsOp_74 In)
  (connect AttributeVectorOp_61 Out AttributeMathsOp_74 In)
  (connect GetAttributeFromPointIndex_70 Attribute AttributeMathsOp_87 InA)
  (connect CreateAttribute_89 Out AttributeMathsOp_87 InB)
  (connect Merge_42 Out AttributePartition_46 In)
  (connect Merge_69 Out AttributePartition_73 In)
  (connect MakeVectorAttribute_37 Out AttributeVectorOp_18 In)
  (connect MakeVectorAttribute_37 Out AttributeVectorOp_18 Transform)
  (connect MakeRotatorAttribute_19 Out AttributeVectorOp_22 InA)
  (connect CreateAttribute_23 Out AttributeVectorOp_22 InB)
  (connect MakeRotatorAttribute_29 Out AttributeVectorOp_33 InA)
  (connect CreateAttribute_35 Out AttributeVectorOp_33 InB)
  (connect MakeVectorAttribute_43 Out AttributeVectorOp_39 In)
  (connect MakeVectorAttribute_43 Out AttributeVectorOp_39 Transform)
  (connect MakeVectorAttribute_49 Out AttributeVectorOp_45 In)
  (connect MakeVectorAttribute_49 Out AttributeVectorOp_45 Transform)
  (connect MakeRotatorAttribute_43 Out AttributeVectorOp_47 InA)
  (connect CreateAttribute_49 Out AttributeVectorOp_47 InB)
  (connect MakeVectorAttribute_61 Out AttributeVectorOp_57 In)
  (connect MakeVectorAttribute_61 Out AttributeVectorOp_57 Transform)
  (connect MakeRotatorAttribute_57 Out AttributeVectorOp_61 InA)
  (connect CreateAttribute_63 Out AttributeVectorOp_61 InB)
  (connect MakeVectorAttribute_69 Out AttributeVectorOp_71 In)
  (connect MakeVectorAttribute_69 Out AttributeVectorOp_71 Transform)
  (connect MakeVectorAttribute_75 Out AttributeVectorOp_77 In)
  (connect MakeVectorAttribute_75 Out AttributeVectorOp_77 Transform)
  (connect MakeVectorAttribute_83 Out AttributeVectorOp_85 In)
  (connect MakeVectorAttribute_83 Out AttributeVectorOp_85 Transform)
  (connect MakeVectorAttribute_87 Out AttributeVectorOp_89 In)
  (connect MakeVectorAttribute_87 Out AttributeVectorOp_89 Transform)
  (connect MakeVectorAttribute_95 Out AttributeVectorOp_97 In)
  (connect MakeVectorAttribute_95 Out AttributeVectorOp_97 Transform)
  (connect MakeVectorAttribute_99 Out AttributeVectorOp_101 In)
  (connect MakeVectorAttribute_99 Out AttributeVectorOp_101 Transform)
  (connect AddTags_118 Out Branch_0 In)
  (connect AttributeCompareOp_4 Out Branch_1 bOutputToB)
  (connect DefaultInputNode InPoint Branch_1 In)
  (connect AddTags_120 Out Branch_2 In)
  (connect AddTags_122 Out Branch_4 In)
  (connect AddTags_124 Out Branch_6 In)
  (connect CreateSpline_75 Out Branch_14 In)
  (connect CreateSpline_48 Out Branch_16 In)
  (connect DensityFilter_24 Out Branch_27 In)
  (connect MakeRotatorAttribute_50 Out Branch_30 In)
  (connect AttributeMathsOp_30 Out BreakVectorAttribute_31 In)
  (connect CreateAttribute_137 Out CopyAttribute_136 Source)
  (connect FilterByTag_125 InsideFilter CopyAttribute_136 Target)
  (connect CreateAttribute_141 Out CopyAttribute_139 Source)
  (connect FilterByTag_128 InsideFilter CopyAttribute_139 Target)
  (connect CreateAttribute_145 Out CopyAttribute_143 Source)
  (connect FilterByTag_132 InsideFilter CopyAttribute_143 Target)
  (connect CreateAttribute_149 Out CopyAttribute_147 Source)
  (connect FilterByTag_134 InsideFilter CopyAttribute_147 Target)
  (connect ToPoint_10 Out CreateSpline_19 In)
  (connect AttributePartition_46 Out CreateSpline_48 In)
  (connect AttributePartition_73 Out CreateSpline_75 In)
  (connect Merge_79 Out CreateSpline_81 In)
  (connect Merge_91 Out CreateSpline_93 In)
  (connect Merge_103 Out CreateSpline_105 In)
  (connect AttributeMathsOp_74 Out DensityFilter_24 In)
  (connect Branch_27 Output%20B FilterByTag_125 In)
  (connect Branch_27 Output%20B FilterByTag_128 In)
  (connect Branch_27 Output%20B FilterByTag_132 In)
  (connect Branch_27 Output%20B FilterByTag_134 In)
  (connect Branch_0 Output%20A FilterByType_8 In)
  (connect Branch_2 Output%20A FilterByType_8 In)
  (connect Branch_4 Output%20A FilterByType_8 In)
  (connect Branch_6 Output%20A FilterByType_8 In)
  (connect Sampled%20Spline%20Points Out FilterDataByTag_25 In)
  (connect Sampled%20Spline%20Points Out FilterDataByTag_39 In)
  (connect Sampled%20Spline%20Points Out FilterDataByTag_53 In)
  (connect Sampled%20Spline%20Points Out FilterDataByTag_67 In)
  (connect SplineSampler_11 Out Gather_11 In)
  (connect SplineSampler_107 Out Gather_11 In)
  (connect Branch_16 Output%20A Gather_19 In)
  (connect SplineSampler_23 Out Gather_19 In)
  (connect Branch_14 Output%20A Gather_21 In)
  (connect SplineSampler_25 Out Gather_21 In)
  (connect CopyAttribute_136 Out Gather_83 In)
  (connect CopyAttribute_139 Out Gather_83 In)
  (connect CopyAttribute_143 Out Gather_83 In)
  (connect CopyAttribute_147 Out Gather_83 In)
  (connect ApplyScaleToBounds_0 Out GetAttributeFromPointIndex_20 In)
  (connect ApplyScaleToBounds_0 Out GetAttributeFromPointIndex_22 In)
  (connect ApplyScaleToBounds_0 Out GetAttributeFromPointIndex_27 In)
  (connect AttributeVectorOp_18 Out GetAttributeFromPointIndex_29 In)
  (connect AttributeVectorOp_39 Out GetAttributeFromPointIndex_41 In)
  (connect AttributeVectorOp_45 Out GetAttributeFromPointIndex_47 In)
  (connect AttributeVectorOp_57 Out GetAttributeFromPointIndex_59 In)
  (connect ApplyScaleToBounds_0 Out GetAttributeFromPointIndex_70 In)
  (connect DefaultInputNode InPoint GetPointsCount_3 In)
  (connect CreateSpline_19 Out Loop_112 SplineIn)
  (connect MakeRotatorAttribute_21 Out MakeRotatorAttribute_19 X)
  (connect FilterDataByTag_25 InsideFilter MakeRotatorAttribute_21 Z)
  (connect MakeRotatorAttribute_31 Out MakeRotatorAttribute_29 X)
  (connect FilterDataByTag_39 InsideFilter MakeRotatorAttribute_31 Z)
  (connect MakeRotatorAttribute_45 Out MakeRotatorAttribute_43 X)
  (connect FilterDataByTag_53 InsideFilter MakeRotatorAttribute_45 Z)
  (connect ToPoint_49 Out MakeRotatorAttribute_50 Z)
  (connect Branch_30 Output%20B MakeRotatorAttribute_55 X)
  (connect MakeRotatorAttribute_59 Out MakeRotatorAttribute_57 X)
  (connect FilterDataByTag_67 InsideFilter MakeRotatorAttribute_59 Z)
  (connect ApplyScaleToBounds_0 Out MakeVectorAttribute_37 X)
  (connect ApplyScaleToBounds_0 Out MakeVectorAttribute_37 Y)
  (connect ApplyScaleToBounds_0 Out MakeVectorAttribute_37 Z)
  (connect ApplyScaleToBounds_0 Out MakeVectorAttribute_43 X)
  (connect ApplyScaleToBounds_0 Out MakeVectorAttribute_43 Y)
  (connect ApplyScaleToBounds_0 Out MakeVectorAttribute_43 Z)
  (connect ApplyScaleToBounds_0 Out MakeVectorAttribute_49 X)
  (connect ApplyScaleToBounds_0 Out MakeVectorAttribute_49 Y)
  (connect ApplyScaleToBounds_0 Out MakeVectorAttribute_49 Z)
  (connect ApplyScaleToBounds_0 Out MakeVectorAttribute_61 X)
  (connect ApplyScaleToBounds_0 Out MakeVectorAttribute_61 Y)
  (connect ApplyScaleToBounds_0 Out MakeVectorAttribute_61 Z)
  (connect ApplyScaleToBounds_0 Out MakeVectorAttribute_69 X)
  (connect ApplyScaleToBounds_0 Out MakeVectorAttribute_69 Y)
  (connect ApplyScaleToBounds_0 Out MakeVectorAttribute_69 Z)
  (connect ApplyScaleToBounds_0 Out MakeVectorAttribute_75 X)
  (connect ApplyScaleToBounds_0 Out MakeVectorAttribute_75 Y)
  (connect ApplyScaleToBounds_0 Out MakeVectorAttribute_75 Z)
  (connect ApplyScaleToBounds_0 Out MakeVectorAttribute_83 X)
  (connect ApplyScaleToBounds_0 Out MakeVectorAttribute_83 Y)
  (connect ApplyScaleToBounds_0 Out MakeVectorAttribute_83 Z)
  (connect ApplyScaleToBounds_0 Out MakeVectorAttribute_87 X)
  (connect ApplyScaleToBounds_0 Out MakeVectorAttribute_87 Y)
  (connect ApplyScaleToBounds_0 Out MakeVectorAttribute_87 Z)
  (connect ApplyScaleToBounds_0 Out MakeVectorAttribute_95 X)
  (connect ApplyScaleToBounds_0 Out MakeVectorAttribute_95 Y)
  (connect ApplyScaleToBounds_0 Out MakeVectorAttribute_95 Z)
  (connect ApplyScaleToBounds_0 Out MakeVectorAttribute_99 X)
  (connect ApplyScaleToBounds_0 Out MakeVectorAttribute_99 Y)
  (connect ApplyScaleToBounds_0 Out MakeVectorAttribute_99 Z)
  (connect SplineSampler_27 Out Merge_42 In)
  (connect SplineSampler_41 Out Merge_42 In)
  (connect Branch_27 Output%20A Merge_46 In)
  (connect Gather_83 Out Merge_46 In)
  (connect SplineSampler_27 Out Merge_69 In)
  (connect SplineSampler_67 Out Merge_69 In)
  (connect AttributeVectorOp_71 Out Merge_79 In)
  (connect AttributeVectorOp_77 Out Merge_79 In)
  (connect AttributeVectorOp_85 Out Merge_91 In)
  (connect AttributeVectorOp_89 Out Merge_91 In)
  (connect AttributeVectorOp_97 Out Merge_103 In)
  (connect AttributeVectorOp_101 Out Merge_103 In)
  (connect Gather_21 Out Projection_13 In)
  (connect WorldRayHitQuery_12 Out Projection_13 Projection%20Target)
  (connect Gather_21 Out Projection_42 In)
  (connect WorldRayHitQuery_35 Out Projection_42 Projection%20Target)
  (connect Merge_46 Out Projection_48 In)
  (connect Input%20Surface Out Projection_48 Projection%20Target)
  (connect Gather_19 Out Projection_72 In)
  (connect WorldRayHitQuery_66 Out Projection_72 Projection%20Target)
  (connect Gather_19 Out Projection_91 In)
  (connect WorldRayHitQuery_82 Out Projection_91 Projection%20Target)
  (connect FilterByType_8 InsideFilter SplineSampler_11 Spline)
  (connect Branch_16 Output%20B SplineSampler_23 Spline)
  (connect Branch_14 Output%20B SplineSampler_25 Spline)
  (connect CreateSpline_81 Out SplineSampler_27 Spline)
  (connect CreateSpline_93 Out SplineSampler_41 Spline)
  (connect CreateSpline_105 Out SplineSampler_67 Spline)
  (connect Loop_112 SplinesOut SplineSampler_107 Spline)
  (connect Branch_0 Output%20B ToPoint_10 In)
  (connect Branch_2 Output%20B ToPoint_10 In)
  (connect Branch_4 Output%20B ToPoint_10 In)
  (connect Branch_6 Output%20B ToPoint_10 In)
  (connect Projection_48 Out ToPoint_49 In)
  (connect Branch_30 Output%20A TransformPoints_56 In)
  (connect MakeRotatorAttribute_55 Out TransformPoints_56 In)
  (connect GetAttributeFromPointIndex_27 Attribute WorldRayHitQuery_12 RayDirection)
  (connect BreakVectorAttribute_31 Y WorldRayHitQuery_12 RayLength)
  (connect GetAttributeFromPointIndex_47 Attribute WorldRayHitQuery_12 RayOrigin)
  (connect AttributeMathsOp_36 Out WorldRayHitQuery_35 RayDirection)
  (connect BreakVectorAttribute_31 Y WorldRayHitQuery_35 RayLength)
  (connect GetAttributeFromPointIndex_59 Attribute WorldRayHitQuery_35 RayOrigin)
  (connect GetAttributeFromPointIndex_70 Attribute WorldRayHitQuery_66 RayDirection)
  (connect BreakVectorAttribute_31 X WorldRayHitQuery_66 RayLength)
  (connect GetAttributeFromPointIndex_29 Attribute WorldRayHitQuery_66 RayOrigin)
  (connect AttributeMathsOp_87 Out WorldRayHitQuery_82 RayDirection)
  (connect BreakVectorAttribute_31 X WorldRayHitQuery_82 RayLength)
  (connect GetAttributeFromPointIndex_41 Attribute WorldRayHitQuery_82 RayOrigin)
  (output OutPoints TransformPoints_56))
