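; PCG subgraph that converts input spatial data's bounds into a single point, plus separate Transform and Extents attribute outputs.
; Get Bounds feeds attribute math that derives the center and half-size from the min and max bounds.
; A created point takes these as Position and Extents, giving it a world-space location with local bounds.
; The Transform output carries only the center as translation, and Delete Attributes trims each attribute output to its named value.
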(graph SpatialDataBoundsToPoint
  (inputs (SpatialData :type Spatial))
  (outputs (Point :type Point) (Transform :type Param) (Extents :type Param))
  ;[0 0 0] Build point with World Transform and Local Bounds from Input Spatial Data World Space Bounds
  (bind GetBounds_0 (Get%20Bounds))
  (bind AttributeMathsOp_1 (Attribute%20Maths%20Op :InputSource1 BoundsMax :InputSource2 BoundsMin :OutputTarget avg))
  (bind CreateAttribute_3 (Create%20Constant :AttributeTypes (struct :DoubleValue 0.5)))
  (bind AttributeMathsOp_2 (Attribute%20Maths%20Op :Operation Multiply))
  (bind CreatePoints_4 (Create%20Points))
  (bind AttributeMathsOp_5 (Attribute%20Maths%20Op :InputSource1 $Position :Operation Set))
  (bind AttributeMathsOp_6 (Attribute%20Maths%20Op :InputSource1 BoundsMax :InputSource2 BoundsMin :Operation Subtract :OutputTarget Extents))
  (bind AttributeMathsOp_7 (Attribute%20Maths%20Op :Operation Multiply))
  (bind AttributeMathsOp_9 (Attribute%20Maths%20Op :InputSource1 $Extents :Operation Set))
  (bind MakeTransformAttribute_0 (Make%20Transform%20Attribute :OutputTarget Transform))
  (bind FilterAttribute_1 (Delete%20Attributes :bTokenizeOnWhiteSpace true :Operation KeepSelectedAttributes :SelectedAttributes "Extents"))
  (bind FilterAttribute_3 (Delete%20Attributes :bTokenizeOnWhiteSpace true :Operation KeepSelectedAttributes :SelectedAttributes "Transform"))
  (connect GetBounds_0 Out AttributeMathsOp_1 InA)
  (connect GetBounds_0 Out AttributeMathsOp_1 InB)
  (connect AttributeMathsOp_1 Out AttributeMathsOp_2 InA)
  (connect CreateAttribute_3 Out AttributeMathsOp_2 InB)
  (connect CreatePoints_4 Out AttributeMathsOp_5 InA)
  (connect AttributeMathsOp_2 Out AttributeMathsOp_5 InB)
  (connect GetBounds_0 Out AttributeMathsOp_6 InA)
  (connect GetBounds_0 Out AttributeMathsOp_6 InB)
  (connect AttributeMathsOp_6 Out AttributeMathsOp_7 InA)
  (connect CreateAttribute_3 Out AttributeMathsOp_7 InB)
  (connect AttributeMathsOp_5 Out AttributeMathsOp_9 InA)
  (connect AttributeMathsOp_7 Out AttributeMathsOp_9 InB)
  (connect AttributeMathsOp_7 Out FilterAttribute_1 In)
  (connect MakeTransformAttribute_0 Out FilterAttribute_3 In)
  (connect DefaultInputNode SpatialData GetBounds_0 In)
  (connect AttributeMathsOp_2 Out MakeTransformAttribute_0 Translation)
  (output Extents FilterAttribute_1)
  (output Point AttributeMathsOp_9)
  (output Transform FilterAttribute_3))
